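# Masked, lit Niagara mesh-particle material for the Birthday pickaxe splat effect.
# Radial UVs pan tiled noise, driving normal-based wobble plus a center-weighted Z pulldown offset, with dynamic parameters controlling pulldown, wobble, pan and reflection brightness.
# Particle color gains a fake reflection from noise sampled by view-space normals plus Fresnel.
# A smoothstep compares radial distance plus particle alpha against panned noise to form an erosion mask.
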Material /BRCosmetics/Weapons/FORT_Melee/Pickaxe_FN_Birthday_Male/FX/M_FX_BirthdayPickaxe_Splat
graph {
 properties {
  bCastRayTracedShadows=false
  bUsedWithNiagaraMeshParticles=true
  blendMode=BLEND_Masked
  pixelDepthOffsetMode=PDOM_Legacy
  usedShadingModels=Default Lit
 }
 nodes [
  n1 = MaterialFunctionCall(Vector or UVs=Subtract(A=TextureCoordinate()) {constB=0.5}) {materialFunction=&VectorToRadialValue}
  n2 = Constant2Vector()
  n3 = DynamicParameter() {paramNames [Pulldown, Wobble, Panner, RefBrightness]}
  n4 = Constant2Vector() {g=1, r=2}
  n5 = ScalarParameter() {defaultValue=1, parameterName=Tiling}
  n6 = Panner(
   Coordinate=Multiply(A=@n1.Radial Coordinates, B=Multiply(A=@n4, B=@n5)) {constB=4}
   Time=DynamicParameter() {paramNames [Pulldown, Wobble, Panner, RefBrightness]}.Panner
  ) {speedY=-1}
  n7 = VertexNormalWS()
  Add(
   A=Multiply(A=OneMinus(None=@n1.Linear Distance), B=AppendVector(A=@n2, B=@n3.Pulldown))
   B=Multiply(A=TextureSample(UVs=@n6) {texture=/Game/Effects/Fort_Effects/Textures/Tile/T_GaussianNoise_01_M}.R, B=Multiply(A=@n7, B=@n3.Wobble))
  )
  n8 = DynamicParameter() {paramNames [Pulldown, Wobble, Panner, RefBrightness]}
  n9 = TextureCoordinate()
  n10 = DDX(Value=@n9)
  n11 = DDY(Value=@n9)
  n12 = TextureSample!(UVs=@n6, Tex, "DDX(UVs)"=@n10, "DDY(UVs)"=@n11, Apply View MipBias) {mipValueMode=TMVM_Derivative, samplerType=SAMPLERTYPE_Normal, texture=/Game/Effects/Fort_Effects/Textures/Tile/T_GaussianNoise_02_N}
  n13 = Multiply(A=@n12.RGB, B=Constant3Vector() {constant=<rgba 4 4 1 0>})
  n14 = Transform(None=@n13) {transformType=TRANSFORM_View}
  n15 = Subtract(A=Multiply(A=ComponentMask(None=@n14) {g=true, r=true}) {constB=0.5}) {constB=-1}
  n16 = ScalarParameter() {defaultValue=1, parameterName=ReflectionTiling}
  n17 = Add(
   A=TextureSample!(UVs=Multiply(A=@n15, B=@n16), Tex, Apply View MipBias) {texture=&T_GaussianNoise_01_M}.R
   B=Fresnel(Normal=@n13) {baseReflectFraction=0, exponent=1}
  )
  n18 = ParticleColor()
  Add(A=@n18.RGB, B=Multiply(A=@n18.RGB, B=Multiply(A=@n8.RefBrightness, B=@n17)))
  Constant()
  Multiply()
  n19 = TextureSample!(UVs=@n6, Tex, "DDX(UVs)"=@n10, "DDY(UVs)"=@n11, Apply View MipBias) {mipValueMode=TMVM_Derivative, texture=&T_GaussianNoise_01_M}
  n20 = ScalarParameter() {parameterName=Erosion_W}
  Reroute(None=SmoothStep(Min=@n19.R, Max=Add(A=@n19.R, B=@n20), Value=Add(A=@n1.Linear Distance, B=@n18.A)))
  ScalarParameter() {parameterName=r}
  ScalarParameter() {parameterName=Erode}
  VectorParameter() {parameterName=Color, defaultValue=<rgba 0 0 0 1>}
  VectorParameter() {parameterName=Pulldown, defaultValue=<rgba 0 0 0 1>}
 ]
}
