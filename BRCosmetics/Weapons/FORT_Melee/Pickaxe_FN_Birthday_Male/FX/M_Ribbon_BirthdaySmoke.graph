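# Unlit translucent Niagara ribbon material for the smoke trail effect on the Birthday pickaxe.
# Particle color RGB feeds the MF_FX_EmissiveAdjustments function to shape the emissive color.
# Particle alpha is multiplied by the green channel of a puffy smoke trail texture, tiled in U and panned along U so the smoke scrolls along the ribbon.
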Material /BRCosmetics/Weapons/FORT_Melee/Pickaxe_FN_Birthday_Male/FX/M_Ribbon_BirthdaySmoke
graph {
 properties {bUsedWithNiagaraRibbons=true, blendMode=BLEND_Translucent, pixelDepthOffsetMode=PDOM_Legacy, shadingModel=MSM_Unlit, usedShadingModels=Unlit}
 nodes [
  n1 = ParticleColor()
  MaterialFunctionCall(Emissive=@n1.RGB) {materialFunction=&MF_FX_EmissiveAdjustments}
  n2 = TextureCoordinate() {uTiling=0.25}
  Multiply(
   A=@n1.A
   B=TextureSample(UVs=Panner(Coordinate=@n2) {bFractionalPart=true, speedX=-0.25}) {texture=/Game/Effects/Fort_Effects/Textures/Smoke/T_Puffy_Smoke_Trail_01}.G
  )
 ]
}
