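# Idle Niagara effect for the FN Birthday 5 pickaxe, showing a socket-attached flame with glow, smoke and colored glitch sprites.
# Flame and Smoke are counter-linked ribbons that follow User.SkeletalMesh briefly, then rise with curl noise and velocity-based drag.
# Glow adds short-lived sprites and Light a flickering orange light.
# GlitchDetails spawns bone-relative cyan/magenta sprites with a swing-driven visibility tag; Fountain, GlitchLarge and Glitches are disabled.
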NiagaraSystem /BRCosmetics/Weapons/FORT_Melee/Pickaxe_FN_Birthday_Male/FX/NS_PickAxe_FNBirthday5_Idle
dependencies {
 usedDataInterfaces [
  &NiagaraDataInterfaceCurve
  &NiagaraDataInterfaceColorCurve
  /Script/Niagara.NiagaraDataInterfaceSimpleCounter
  &NiagaraDataInterfaceArrayColor
  &NiagaraDataInterfaceArrayFloat
  &NiagaraDataInterfaceVector2DCurve
 ]
 usedDynamicInputs []
 usedModules [&SystemState]
}
emitters {
 Flame {
  emitterUpdateScript [
   &EmitterState()
   &SpawnRate(SpawnRate=50, Use Spawn Probability=false)
  ]
  particleSpawnScript [
   &InitializeParticle(UsePositionOffset=false, Write Material Random=true)
   &InitializeRibbon(
    Color=<rgba 10 1.444 0 0.3>
    Lifetime=&RandomRangeFloat(Minimum=0.15, Maximum=0.3)
    Module.Write Color=true
    Module.Write Lifetime=true
    Module.Write Mass=true
    Module.Write Material Random=true
    Module.Write Position=true
    Module.Write RibbonTwist=false
    Module.Write RibbonWidth=true
    Ribbon Width=&RandomRangeFloat(Minimum=1, Maximum=2.5)
   )
   &SkeletalMeshLocation(
    Inherit Velocity=true
    Interpolate Mesh Sampling=false
    MaskSpawnGroup=false
    Mesh Sampling Type=::NewEnumerator5
    Sampled Position Offset Coordinate Space:ENiagaraMeshTransforms=::NewEnumerator3
    Skeletal Mesh=$User.SkeletalMesh
    Socket Sampling Mode=::NewEnumerator5
    VelocitySpeedLimit=false
   )
   /SetVariables(Particles.RibbonLinkOrder:NiagaraFloat=&MakeFloatFromInt(Int=/Niagara/DynamicInputs/DataInterface/GenerateIntFromCounter(SimpleCounter:NiagaraDataInterfaceSimpleCounter=DataInterface() {propertyValues {GpuSyncMode=None, InitialValue=0}}))) {isSetParameters=true, module=SetVariables_#1}
  ]
  particleUpdateScript [
   &ParticleState()
   &SkeletalMeshLocation(
    Apply To Particle Position:NiagaraBool=&SetBoolByFloatComparison(A=$Particles.NormalizedAge, B=0.2, Comparison Type=::NewEnumerator6)
    Inherit Velocity=false
    Interpolate Mesh Sampling=false
    MaskSpawnGroup=false
    Mesh Sampling Type=::NewEnumerator5
    Sampled Position Offset=&FN_FrontEndVector3Modifier(Gameplay=<0 0 -1>, Front End=<0 0 -1>)
    Skeletal Mesh=$User.SkeletalMesh
    Socket Sampling Mode=::NewEnumerator5
   ) {module=SkeletalMeshLocation001}
   &GravityForce(Gravity=<0 0 500>)
   /Niagara/Modules/Update/Forces/CurlNoiseForce(Noise Strength=50, Pan Noise=false, Random Seed:NiagaraInt32=&RandomRangeInt(Minimum=0, Maximum=100))
   &Drag(
    Drag=&Multiply_Float(A=/Niagara/DynamicInputs/TypeConversions/VectorLengthSafe(Vector:Vector3f=$Output.SkeletalMeshLocation001.SampledVelocity), B=0.9)
    Use Linear Drag=true
    Use Rotational Drag=false
   )
   &SolveForcesAndVelocity(Clamp Velocity=false, Limit Acceleration=false)
   &ScaleColor(
    Scale Alpha=&FloatFromCurve(FloatCurve=DataInterface() {propertyValues {Curve {keys [<at 0 1 RCIM_Cubic>, {arriveTangent=-1, interpMode=RCIM_Cubic, leaveTangent=-1, time=0.2, value=1}, <at 1 0 RCIM_Cubic>]}}})
    Scale RGB=&VectorFromFloat(Value=&FN_FrontEndFloatModifier(Gameplay=1, Front End=2))
    ScaleA=true
    ScaleRGB=true
   )
   /SetVariables(
    Particles.DistanceScale:NiagaraFloat=&Add_Float(
     A=1
     B=&ClampFloat(
      Float=&Multiply_Float(
       A=-0.1
       B=&VectorLength(Vector=&Subtract_Vector(
         A=Linked() {linkedVariable {name=Particles.Position, type=NiagaraPosition}}
         B=Linked() {linkedVariable {name=Particles.Initial.Position, type=NiagaraPosition}}
        ))
      )
      Min=-0.6
      Max=0
     )
    )
    Particles.VelocityScale:NiagaraFloat=&Add_Float(
     A=10
     B=&Multiply_Float(A=-0.005, B=&VectorLengthSafe(Vector:Vector3f=$Output.SkeletalMeshLocation001.SampledVelocity))
    )
   ) {isSetParameters=true}
   &KillParticlesInVolume(
    Invert Volume=true
    Origin Offset:Vector3f=&Multiply_VectorByFloat(Vector=$Output.SkeletalMeshLocation001.SampledVelocity, Float=0.01)
    Sphere Radius=10
    Volume Origin:Vector3f=$Output.SkeletalMeshLocation001.SampledPosition
   ) {module=KillParticlesInVolume002}
   &ScaleRibbonWidth(Ribbon Width Scale=&FloatFromCurve(
     FloatCurve=DataInterface() {propertyValues {Curve {keys [
         {arriveTangent=6.0445, interpMode=RCIM_Cubic, leaveTangent=6.0445, tangentMode=RCTM_User, time=0, value=1}
         {arriveTangent=-1, interpMode=RCIM_Cubic, leaveTangent=-1, time=0.2036, value=1.9766}
         <at 1 0>
        ]}}}
     Scale Curve=&MakeCustomFloatFromBool(
      Bool=&SetBoolByFloatComparison(A=$Particles.DistanceScale, B=0.5)
      True Float=1
      False Float=$Particles.DistanceScale
     )
    ))
  ]
  rendererClasses [&NiagaraRibbonRendererProperties]
  renderers [{rendererClass=&NiagaraRibbonRendererProperties}]
 }
 Fountain {
  bEnabled=false
  emitterUpdateScript [
   &EmitterState(Life Cycle Mode=::NewEnumerator1, Loop Duration=0.7, UseLoopDelay=false)
   &SpawnRate(SpawnRate=&RandomRangeFloat(Minimum=2, Maximum=6), Use Spawn Probability=false)
  ]
  particleSpawnScript [
   &InitializeParticle(
    Color=&SelectLinearColorFromArray(
     Array Sampling Mode=::NewEnumerator0
     Direct Array Index=?
     Color Selection Array=DataInterface() {propertyValues {ColorData [<rgba 0.0467 0.4621 0.5457 0.7>, <rgba 0.5089 0.0006 0.5972 0.7>, <rgba 0.063 0.5149 0.3916 0.7>, <rgba 0.6106 0.0678 0.746 0.7>]}}
     Random Seed=0
    )
    Lifetime Max=0.35
    Lifetime Min=0.1
    Lifetime Mode=::NewEnumerator1
    Mass Mode=::NewEnumerator3
    Sprite Rotation Mode=::NewEnumerator0
    Sprite Size Max=<2.75 1.5>
    Sprite Size Min=<1 1>
    Sprite Size Mode=::NewEnumerator2
    UsePositionOffset=false
    Write Material Random=true
   )
   &FN_SetSpawnGroup(Number of Spawn Groups=2)
   &SkeletalMeshLocation(
    Bone / Socket Index=$Particles.SpawnGroup
    MaskSpawnGroup=false
    Mesh Sampling Type=::NewEnumerator5
    Skeletal Mesh=$User.SkeletalMesh
    Socket Sampling Mode=::NewEnumerator5
   )
   &ShapeLocation(
    Additional Quaternion Rotation=true
    Additional Yaw / Pitch / Roll=<0 0 90>
    Box / Plane Mode:ENiagara_BoxPlaneMode=::NewEnumerator1
    MaskSpawnGroup=false
    Offset=<1 0 0>
    Offset Coordinate Space=::Local
    Offset Mode=::NewEnumerator0
    Override Seed=false
    Plane Midpoint:Vector2f=<0.5 0>
    Plane Size:Vector2f=<7 10>
    Rotation Angle Type:ENiagara_AngleInput=::NewEnumerator0
    Rotation Mode=::NewEnumerator3
    Rotation Quaternion=$StackContext.SkeletalMeshLocation.SampledRotation
    Shape Origin=$StackContext.SkeletalMeshLocation.SampledPosition
    Shape Primitive=::NewEnumerator2
    Transform Order:ENiagara_TransformOrder=::NewEnumerator1
    UseOwnerScale=true
   )
   &AddVelocity(Velocity=<1.1 0 -2>, Velocity Speed Scale=&RandomRangeFloat(Minimum=2, Maximum=9))
  ]
  particleUpdateScript [
   &ParticleState()
   &GravityForce() {enabled=false}
   &Drag(Drag=3, Use Linear Drag=true, Use Rotational Drag=false)
   &ScaleSpriteSize(
    Uniform Curve Scale=1.8
    Uniform Curve Sprite Scale=DataInterface() {propertyValues {Curve {keys [
        <at 0 0 RCIM_Cubic>
        {arriveTangent=1, interpMode=RCIM_Cubic, leaveTangent=1, time=0.1, value=1.05}
        {arriveTangent=-1.1667, interpMode=RCIM_Cubic, leaveTangent=-1.1667, time=0.8, value=0.8}
        <at 1 0 RCIM_Cubic>
       ]}}}
   )
   &ScaleColor(
    Scale Alpha=&FloatFromCurve(FloatCurve=DataInterface() {propertyValues {Curve {keys [
         <at -0.0046 1.0113 RCIM_Cubic>
         {arriveTangent=-1.0036, interpMode=RCIM_Cubic, leaveTangent=-1.0036, time=0.7, value=0.85}
         <at 1.0011 0.002 RCIM_Cubic>
        ]}}})
    Scale RGB=&VectorFromFloat(Value=&FloatFromCurve(FloatCurve=DataInterface() {propertyValues {Curve {keys [<at 0 1>,<at 0.3 1.625>,<at 0.4 0.35>,<at 0.7 1.75>,<at 1 1>]}}}))
    ScaleA=true
    ScaleRGB=true
   )
   /SetVariables(Particles.BackingSpriteSize:Vector2f=<8 2.5>) {isSetParameters=true, module=SetVariables_#2}
   &InheritSourceMovement(
    Applied Position Scale=&VectorFromFloat(Value=&FloatFromCurve(FloatCurve=DataInterface() {propertyValues {Curve {keys [<at 0 1 RCIM_Cubic>, {arriveTangent=-1, interpMode=RCIM_Cubic, leaveTangent=-1, time=0.8, value=0.85}, <at 1 0 RCIM_Cubic>]}}}))
    ApplyForce=false
    ApplyPosition=true
    ApplyPositionFalloff=false
   )
   &SolveForcesAndVelocity(Clamp Velocity=false, Limit Acceleration=false)
   &DynamicMaterialParameters(
    Index 0 Param 1=&Lerp_Float(
     A=1.25
     B=0.5
     Alpha=&FloatFromCurve(FloatCurve=DataInterface() {propertyValues {Curve {keys [{interpMode=RCIM_Cubic, tangentMode=RCTM_User, time=0, value=1}, {interpMode=RCIM_Cubic, tangentMode=RCTM_User, time=1, value=0}]}}})
    )
    Index 0 Param 2=&RandomRangeFloat(Minimum=-1, Maximum=1)
    Index 0 Param 3=&RandomRangeFloat(Minimum=-3, Maximum=3)
    Index 0 Param 4=20
    Module.Param0WriteEnabled=true
    Module.Param1WriteEnabled=true
    Module.Param2WriteEnabled=true
    Module.Param3WriteEnabled=true
   )
  ]
  rendererClasses [&NiagaraSpriteRendererProperties]
  renderers [
   {rendererClass=&NiagaraSpriteRendererProperties}
   {rendererClass=&NiagaraSpriteRendererProperties, rendererIndex=1}
  ]
 }
 GlitchDetails {
  emitterUpdateScript [
   &EmitterState(Life Cycle Mode=::NewEnumerator1, Loop Duration=1, UseLoopDelay=false)
   &SpawnRate(SpawnRate=&RandomRangeFloat(Minimum=12, Maximum=25), Use Spawn Probability=false)
  ]
  particleSpawnScript [
   &InitializeParticle(
    Color=&SelectLinearColorFromArray(
     Array Sampling Mode=::NewEnumerator0
     Direct Array Index=?
     Color Selection Array=DataInterface() {propertyValues {ColorData [<rgba 0.0467 0.4621 0.5457 0.7>, <rgba 0.5089 0.0006 0.5972 0.7>, <rgba 0.063 0.5149 0.3916 0.7>, <rgba 0.6106 0.0678 0.746 0.7>]}}
     Random Seed=0
    )
    Lifetime Max=0.35
    Lifetime Min=0.1
    Lifetime Mode=::NewEnumerator1
    Mass Mode=::NewEnumerator3
    Sprite Rotation Mode=::NewEnumerator0
    Sprite Size Max=<2.75 1.5>
    Sprite Size Min=<1 1>
    Sprite Size Mode=::NewEnumerator2
    UsePositionOffset=false
    Write Material Random=true
   )
   &FN_SetSpawnGroup()
   &SkeletalMeshLocation(Bone / Socket Index=$Particles.SpawnGroup, Bone Sampling Mode=::NewEnumerator4, MaskSpawnGroup=false, Skeletal Mesh=$User.SkeletalMesh)
   &ShapeLocation(
    Additional Quaternion Rotation=true
    Additional Yaw / Pitch / Roll=<35 0 90>
    Box / Plane Mode:ENiagara_BoxPlaneMode=::NewEnumerator1
    MaskSpawnGroup=false
    Offset=<-10 0 0>
    Offset Mode=::NewEnumerator0
    Override Seed=false
    Plane Size:Vector2f=<10 10>
    Rotation Angle Type:ENiagara_AngleInput=::NewEnumerator0
    Rotation Mode=::NewEnumerator3
    Rotation Quaternion=$StackContext.SkeletalMeshLocation.SampledRotation
    Shape Primitive=::NewEnumerator2
    UseOwnerScale=true
   )
   /SetVariables(Particles.Position=/Niagara/DynamicInputs/Math/TransformVectorByMatrix(
     Vector:Vector3f=/Niagara/DynamicInputs/Math/LerpMultipleVectors(
      Alpha:NiagaraFloat=&RandomRangeFloat(Minimum=0, Maximum=1)
      V 1:Vector3f=<8 2 0>
      V 2:Vector3f=<2 -15.9446 0>
      V 3:Vector3f=<-8 2 1>
      V 4:Vector3f=?
      V 5:Vector3f=?
      Vector Count:ENiagara_MultipleLerpCount=::NewEnumerator0
     )
     Transformation Matrix:NiagaraMatrix=/Niagara/DynamicInputs/Math/MakeTransform(
      Inverse:NiagaraBool=false
      Position:Vector3f=$StackContext.SkeletalMeshLocation.SampledPosition
      Local Offset:Vector3f=&RandomRangeVector(Minimum=<0 0 -3>, Maximum=<0 0 3>)
      Scale:Vector3f=<1 1 1>
      Rotation Interface:ENiagara_RotationDefinitionApproach=::NewEnumerator1
      Axis:Vector3f=?
      Quaternion:Quat4f=&MultiplyQuaternion(
       Quaternion A=$StackContext.SkeletalMeshLocation.SampledRotation
       Quaternion B=&MakeQuatFromAxisAngle(
        Axis=<0 0 1>
        AngleInDegrees=&SelectFloatFromArray(Direct Array Index=$Particles.SpawnGroup, Float Selection Array=DataInterface() {propertyValues {FloatData [180,0,180,0]}})
       )
      )
      Rotation Matrix:NiagaraMatrix=?
      X Vector:Vector3f=?
      Yaw:NiagaraFloat=?
      Angle:NiagaraFloat=?
      Pitch:NiagaraFloat=?
      Y Vector:Vector3f=?
      Roll:NiagaraFloat=?
      Z Vector:Vector3f=?
     )
     Transform Vector as Position:NiagaraBool=true
    )) {isSetParameters=true}
   /SetVariables() {isSetParameters=true}
   &AddVelocity(
    Cone Angle=70
    Cone Axis=&RotateVectorByQuaternion(
     Vector To Rotate=<0 1 0>
     Quaternion=&MultiplyQuaternion(
      Quaternion A=$StackContext.SkeletalMeshLocation.SampledRotation
      Quaternion B=&MakeQuatFromAxisAngle(
       Axis=<1 0 0>
       AngleInDegrees=&SelectFloatFromArray(Direct Array Index=$Particles.SpawnGroup, Float Selection Array=DataInterface() {propertyValues {FloatData [0,180,0,180]}})
      )
     )
    )
    Distribution Along Cone Axis=0.25
    Module.UseConeFalloff=true
    Module.UseVelDistribution=true
    Override Seed=false
    Speed Falloff From Cone Axis=0.5
    Velocity Mode=::NewEnumerator2
    Velocity Speed=&RandomRangeFloat(Minimum=20, Maximum=30)
   )
   &AddVelocity(Velocity=<0 0 -1>, Velocity Speed Scale=&RandomRangeFloat(Minimum=15, Maximum=20)) {module=AddVelocity001}
  ]
  particleUpdateScript [
   &ParticleState()
   &ScaleSpriteSize(
    Uniform Curve Scale=1.35
    Uniform Curve Sprite Scale=DataInterface() {propertyValues {Curve {keys [
        <at 0 0 RCIM_Cubic>
        {arriveTangent=1.125, interpMode=RCIM_Cubic, leaveTangent=1.125, time=0.1, value=0.9}
        {arriveTangent=-1, interpMode=RCIM_Cubic, leaveTangent=-1, time=0.8, value=0.9}
        <at 1 0 RCIM_Cubic>
       ]}}}
   )
   &GravityForce() {enabled=false}
   &SkeletalMeshLocation(
    Bone / Socket Index=$Particles.SpawnGroup
    Bone Sampling Mode=::NewEnumerator4
    "Interpolate Mesh Sampling (Update)"=true
    MaskSpawnGroup=false
    Position Sampling=::NewEnumerator1
    Sampled Position Coordinate Space=::World
    Sampled Position Offset=&RotateVectorByQuaternion(
     Vector To Rotate=&Add_Vector(
      A=&LerpMultipleVectors(
       Alpha:NiagaraFloat=&RandomRangeFloat(Minimum=0, Maximum=1)
       V 1:Vector3f=<7 0 0>
       V 2:Vector3f=<2 -15.7611 0>
       V 3:Vector3f=<-8 0 1>
       V 4:Vector3f=?
       V 5:Vector3f=?
       Vector Count:ENiagara_MultipleLerpCount=::NewEnumerator0
      )
      B=&RandomRangeVector(Minimum=<0 0 -3>, Maximum=<0 0 3>)
     )
     Quaternion=&MakeQuatFromAxisAngle(
      Axis=<0 0 1>
      AngleInDegrees=&SelectFloatFromArray(Direct Array Index=$Particles.SpawnGroup, Float Selection Array=DataInterface() {propertyValues {FloatData [180,0,180,0]}})
     )
    )
    Skeletal Mesh=$User.SkeletalMesh
    Source Movement Scale:Vector3f=&VectorFromFloat(Value=0)
    Write To Particle Position:NiagaraBool=&SetBoolByFloatComparison(A=$Particles.NormalizedAge, B=1, Comparison Type=::NewEnumerator6)
   ) {enabled=false, module=SkeletalMeshLocation001}
   /SetVariables(Particles.Position=&Lerp_Vector(
     A=Linked() {linkedVariable {name=Particles.Position, type=NiagaraPosition}}
     B=Linked() {linkedVariable {name=StackContext.SkeletalMeshLocation001.SampledPosition, type=NiagaraPosition}}
     Alpha=0.2
    )) {enabled=false, isSetParameters=true}
   /SetVariables(Particles.Presolve.Position:Vector3f=$StackContext.SkeletalMeshLocation.SampledPosition) {enabled=false, isSetParameters=true}
   /SetVariables(Particles.Position:Vector3f=$StackContext.SkeletalMeshLocation.SampledPosition) {enabled=false, isSetParameters=true}
   /SetVariables(Particles.VisibilityTag=&MakeCustomIntFromBool(Bool=$User.bisSwinging, True INT=1, False INT=0)) {isSetParameters=true}
   &Drag(Drag=5.5, Use Linear Drag=true, Use Rotational Drag=false)
   &ScaleColor(
    Scale Alpha=&FloatFromCurve(FloatCurve=DataInterface() {propertyValues {Curve {keys [
         <at -0.0046 1.0113 RCIM_Cubic>
         {arriveTangent=-1.0036, interpMode=RCIM_Cubic, leaveTangent=-1.0036, time=0.7, value=0.85}
         <at 1.0011 0.002 RCIM_Cubic>
        ]}}})
    Scale RGB=&VectorFromFloat(Value=&FloatFromCurve(FloatCurve=DataInterface() {propertyValues {Curve {keys [<at 0 1>,<at 0.5 1.625>,<at 1 1>]}}}))
    ScaleA=true
    ScaleRGB=true
   ) {enabled=false, module=ScaleColor001}
   &AccelerationForce(Acceleration=&RandomRangeVector(Minimum=<0 0 0>, Maximum=<0 0 10>))
   &InheritVelocity(
    Inherited Velocity Amount Scale=&VectorFromFloat(Value=&FloatFromCurve(
      FloatCurve=DataInterface() {propertyValues {Curve {keys [{interpMode=RCIM_Cubic, tangentMode=RCTM_User, time=0, value=1}, {interpMode=RCIM_Cubic, tangentMode=RCTM_User, time=1, value=0}]}}}
      Scale Curve=0.1
     ))
    Speed Limit=true
    UseSpeedThreshold=false
    Velocity Source=&TransformVector(Vector=$Engine.Owner.Velocity, Source Space=::World, Destination Space=::Local)
   )
   &SolveForcesAndVelocity(Clamp Velocity=false, Limit Acceleration=false)
   &ScaleSpriteSize(
    Scale Sprite Size Mode=::NewEnumerator0
    Uniform Scale Factor=&RemapRange(
     Input Value To Remap=&VectorLength(Vector=&Subtract_Vector(A=Linked() {linkedVariable {name=Particles.Position, type=NiagaraPosition}}, B=$Particles.Initial.SkeletalMeshLocation.SampledPosition))
     Input Min=100
     Input Max=150
     Output Min=1
     Output Max=0
    )
   ) {enabled=false, module=ScaleSpriteSize001}
   /SetVariables(Particles.BackingSpriteSize:Vector2f=<8 3>) {enabled=false, isSetParameters=true}
   &DynamicMaterialParameters(
    Index 0 Param 1=0.2
    Index 0 Param 2=&RandomRangeFloat(Minimum=0, Maximum=5)
    Index 0 Param 3=&RandomRangeFloat(Minimum=0, Maximum=1)
    Index 0 Param 4=&RandomRangeFloat(Minimum=0, Maximum=1.2)
    Module.Param0WriteEnabled=true
    Module.Param1WriteEnabled=true
    Module.Param2WriteEnabled=true
    Module.Param3WriteEnabled=true
   ) {enabled=false}
   &ScaleColor(
    Scale Alpha=&FloatFromCurve(FloatCurve=DataInterface() {propertyValues {Curve {keys [
         <at -0.0046 1.0113 RCIM_Cubic>
         {arriveTangent=-1.0036, interpMode=RCIM_Cubic, leaveTangent=-1.0036, time=0.7, value=0.85}
         <at 1.0011 0.002 RCIM_Cubic>
        ]}}})
    Scale RGB=&VectorFromFloat(Value=&FloatFromCurve(FloatCurve=DataInterface() {propertyValues {Curve {keys [<at 0 1>,<at 0.3 1.625>,<at 0.4 0.35>,<at 0.7 1.75>,<at 1 1>]}}}))
    ScaleA=true
    ScaleRGB=true
   )
   /SetVariables(Particles.BackingSpriteSize:Vector2f=<8 2.5>) {isSetParameters=true}
   &DynamicMaterialParameters(
    Index 0 Param 1=&Lerp_Float(
     A=1.25
     B=0.5
     Alpha=&FloatFromCurve(FloatCurve=DataInterface() {propertyValues {Curve {keys [{interpMode=RCIM_Cubic, tangentMode=RCTM_User, time=0, value=1}, {interpMode=RCIM_Cubic, tangentMode=RCTM_User, time=1, value=0}]}}})
    )
    Index 0 Param 2=&RandomRangeFloat(Minimum=-1, Maximum=1)
    Index 0 Param 3=&RandomRangeFloat(Minimum=-3, Maximum=3)
    Index 0 Param 4=20
    Module.Param0WriteEnabled=true
    Module.Param1WriteEnabled=true
    Module.Param2WriteEnabled=true
    Module.Param3WriteEnabled=true
   ) {module=DynamicMaterialParameters001}
   &CameraOffset(Camera Offset Amount=&FloatFromCurve(
     FloatCurve=DataInterface() {propertyValues {Curve {keys [{interpMode=RCIM_Cubic, tangentMode=RCTM_User, time=0, value=0}, {interpMode=RCIM_Cubic, tangentMode=RCTM_User, time=1, value=1}]}}}
     Scale Curve=5
    ))
  ]
  rendererClasses [&NiagaraSpriteRendererProperties]
  renderers [
   {rendererClass=&NiagaraSpriteRendererProperties}
   {rendererClass=&NiagaraSpriteRendererProperties, rendererIndex=1}
   {rendererClass=&NiagaraSpriteRendererProperties, rendererIndex=2}
  ]
 }
 GlitchLarge {
  bEnabled=false
  emitterUpdateScript [
   &EmitterState(Life Cycle Mode=::NewEnumerator1, Loop Duration=2, UseLoopDelay=false)
   &SpawnRate(SpawnRate=20, Use Spawn Probability=false)
  ]
  particleSpawnScript [
   &InitializeParticle(
    Color=&SelectLinearColorFromArray(
     Array Sampling Mode=::NewEnumerator0
     Direct Array Index=?
     Color Selection Array=DataInterface() {propertyValues {ColorData [<rgba 0.0467 0.4621 0.5457 0.7>, <rgba 0.5089 0.0006 0.5972 0.7>, <rgba 0.063 0.5149 0.3916 0.7>, <rgba 0.6106 0.0678 0.746 0.7>]}}
     Random Seed=0
    )
    Lifetime Max=0.2
    Lifetime Min=0.1
    Lifetime Mode=::NewEnumerator1
    Mass Mode=::NewEnumerator3
    Sprite Rotation Mode=::NewEnumerator0
    Sprite Size Mode=::NewEnumerator1
    Uniform Sprite Size Max=15
    Uniform Sprite Size Min=12
    UsePositionOffset=false
    Write Material Random=true
   )
   &ShapeLocation(
    Additional Quaternion Rotation=true
    Additional Yaw / Pitch / Roll=<35 0 90>
    Box / Plane Mode:ENiagara_BoxPlaneMode=::NewEnumerator1
    MaskSpawnGroup=false
    Offset=<-10 0 0>
    Offset Mode=::NewEnumerator0
    Override Seed=false
    Plane Size:Vector2f=<10 10>
    Rotation Angle Type:ENiagara_AngleInput=::NewEnumerator0
    Rotation Mode=::NewEnumerator3
    Rotation Quaternion=$StackContext.SkeletalMeshLocation.SampledRotation
    Shape Primitive=::NewEnumerator2
    UseOwnerScale=true
   ) {enabled=false}
  ]
  particleUpdateScript [
   &ParticleState()
   &SkeletalMeshLocation(Bone Sampling Mode=::NewEnumerator0, MaskSpawnGroup=false, OverrideRandomSeed=false, Skeletal Mesh=$User.SkeletalMesh)
   &ScaleSpriteSize(
    Non-Uniform Curve Scale=<1.2 0.45>
    Non-Uniform Curve Sprite Scale=DataInterface() {propertyValues {
      XCurve {keys [{tangentMode=RCTM_User, time=0, value=0.5}, {tangentMode=RCTM_User, time=1, value=1.375}]}
      YCurve {keys [{tangentMode=RCTM_User, time=0, value=1}, {tangentMode=RCTM_User, time=1, value=1}]}
     }}
    Scale Sprite Size Mode=::NewEnumerator3
   )
   &GravityForce() {enabled=false}
   &Drag(Drag=0.25, Use Linear Drag=true, Use Rotational Drag=false)
   &ScaleColor(
    Scale Alpha=&FloatFromCurve(FloatCurve=DataInterface() {propertyValues {Curve {keys [<at -0.0046 1.0113 RCIM_Cubic>,<at 1.0011 0.002 RCIM_Cubic>]}}})
    Scale RGB=&VectorFromFloat(Value=&FloatFromCurve(FloatCurve=DataInterface() {propertyValues {Curve {keys [<at 0 1>,<at 0.3 1.625>,<at 0.4 0.35>,<at 0.7 1.75>,<at 1 1>]}}}))
    ScaleA=true
    ScaleRGB=true
   ) {module=ScaleColor001}
   &SolveForcesAndVelocity(Clamp Velocity=false, Limit Acceleration=false)
   &DynamicMaterialParameters(
    Index 0 Param 1=1
    Index 0 Param 2=&RandomRangeFloat(Minimum=0, Maximum=3)
    Index 0 Param 3=&RandomRangeFloat(Minimum=0, Maximum=1)
    Index 0 Param 4=&RandomRangeFloat(Minimum=0, Maximum=1.2)
    Module.Param0WriteEnabled=true
    Module.Param1WriteEnabled=true
    Module.Param2WriteEnabled=true
    Module.Param3WriteEnabled=true
   )
   &CameraOffset()
  ]
  rendererClasses [&NiagaraSpriteRendererProperties]
  renderers [{rendererClass=&NiagaraSpriteRendererProperties}]
 }
 Glitches {
  bEnabled=false
  emitterUpdateScript [
   &EmitterState(Life Cycle Mode=::NewEnumerator1, Loop Duration=0.7, UseLoopDelay=false)
   &SpawnRate(SpawnRate=&RandomRangeFloat(Minimum=30, Maximum=30), Use Spawn Probability=false)
  ]
  particleSpawnScript [
   &InitializeParticle(
    Color=&SelectLinearColorFromArray(
     Array Sampling Mode=::NewEnumerator0
     Direct Array Index=?
     Color Selection Array=DataInterface() {propertyValues {ColorData [<rgba 0.0467 0.4621 0.5457 0.7>, <rgba 0.5089 0.0006 0.5972 0.7>, <rgba 0.063 0.5149 0.3916 0.7>, <rgba 0.6106 0.0678 0.746 0.7>]}}
     Random Seed=0
    )
    Lifetime Max=0.4
    Lifetime Min=0.125
    Lifetime Mode=::NewEnumerator1
    Mass Mode=::NewEnumerator3
    Sprite Rotation Mode=::NewEnumerator0
    Sprite Size Max=<2 2>
    Sprite Size Min=<1 1>
    Sprite Size Mode=::NewEnumerator2
    UsePositionOffset=false
    Write Material Random=true
   )
   &FN_SetSpawnGroup()
   &SkeletalMeshLocation(Bone / Socket Index=$Particles.SpawnGroup, Bone Sampling Mode=::NewEnumerator4, MaskSpawnGroup=false, Skeletal Mesh=$User.SkeletalMesh)
   &AddVelocity(Velocity=<1.1 0 -2>, Velocity Speed Scale=&RandomRangeFloat(Minimum=2, Maximum=9))
   &ShapeLocation(
    Additional Quaternion Rotation=true
    Additional Yaw / Pitch / Roll=<35 0 90>
    MaskSpawnGroup=false
    Offset=<-10 0 0>
    Offset Mode=::NewEnumerator0
    Override Seed=false
    Rotation Angle Type:ENiagara_AngleInput=::NewEnumerator0
    Rotation Mode=::NewEnumerator3
    Rotation Quaternion=$StackContext.SkeletalMeshLocation.SampledRotation
    Shape Primitive=::NewEnumerator2
    Surface Only=false
    UseOwnerScale=true
   )
   /SetVariables(Particles.Position=&TransformVectorByMatrix(
     Vector:Vector3f=&LerpMultipleVectors(
      Alpha:NiagaraFloat=&RandomRangeFloat(Minimum=0, Maximum=1)
      V 1:Vector3f=<7 0 0>
      V 2:Vector3f=<2 -15.7611 0>
      V 3:Vector3f=<-8 0 1>
      V 4:Vector3f=?
      V 5:Vector3f=?
      Vector Count:ENiagara_MultipleLerpCount=::NewEnumerator0
     )
     Transformation Matrix:NiagaraMatrix=&MakeTransform(
      Inverse:NiagaraBool=false
      Position:Vector3f=$StackContext.SkeletalMeshLocation.SampledPosition
      Local Offset:Vector3f=&RandomRangeVector(Minimum=<0 0 -2>, Maximum=<0 0 2>)
      Scale:Vector3f=<1 1 1>
      Rotation Interface:ENiagara_RotationDefinitionApproach=::NewEnumerator1
      Axis:Vector3f=?
      Quaternion:Quat4f=&MultiplyQuaternion(
       Quaternion A=$StackContext.SkeletalMeshLocation.SampledRotation
       Quaternion B=&MakeQuatFromAxisAngle(
        Axis=<0 0 1>
        AngleInDegrees=&SelectFloatFromArray(Direct Array Index=$Particles.SpawnGroup, Float Selection Array=DataInterface() {propertyValues {FloatData [180,0,180,0]}})
       )
      )
      Rotation Matrix:NiagaraMatrix=?
      X Vector:Vector3f=?
      Yaw:NiagaraFloat=?
      Angle:NiagaraFloat=?
      Pitch:NiagaraFloat=?
      Y Vector:Vector3f=?
      Roll:NiagaraFloat=?
      Z Vector:Vector3f=?
     )
     Transform Vector as Position:NiagaraBool=true
    )) {isSetParameters=true}
   &AddVelocity(
    Distribution Along Cone Axis=0.25
    Module.UseConeFalloff=true
    Module.UseVelDistribution=true
    Override Seed=false
    Speed Falloff From Cone Axis=0.5
    Velocity Mode=::NewEnumerator2
    Velocity Speed=&RandomRangeFloat(Minimum=20, Maximum=30)
   ) {enabled=false, module=AddVelocity001}
   &AddVelocity(Velocity=<0 0 -1>, Velocity Speed Scale=&RandomRangeFloat(Minimum=0, Maximum=8)) {module=AddVelocity002}
  ]
  particleUpdateScript [
   &ParticleState()
   &GravityForce() {enabled=false}
   &Drag(Drag=3, Use Linear Drag=true, Use Rotational Drag=false)
   &ScaleSpriteSize(
    Uniform Curve Scale=1.8
    Uniform Curve Sprite Scale=DataInterface() {propertyValues {Curve {keys [
        <at 0 0 RCIM_Cubic>
        {arriveTangent=1, interpMode=RCIM_Cubic, leaveTangent=1, time=0.1, value=1.05}
        {arriveTangent=-1.1667, interpMode=RCIM_Cubic, leaveTangent=-1.1667, time=0.8, value=0.8}
        <at 1 0 RCIM_Cubic>
       ]}}}
   )
   &ScaleColor(
    Scale Alpha=&FloatFromCurve(FloatCurve=DataInterface() {propertyValues {Curve {keys [
         <at -0.0046 1.0113 RCIM_Cubic>
         {arriveTangent=-1.0036, interpMode=RCIM_Cubic, leaveTangent=-1.0036, time=0.7, value=0.85}
         <at 1.0011 0.002 RCIM_Cubic>
        ]}}})
    Scale RGB=&VectorFromFloat(Value=&FloatFromCurve(FloatCurve=DataInterface() {propertyValues {Curve {keys [<at 0 1>,<at 0.3 1.625>,<at 0.4 0.35>,<at 0.7 1.75>,<at 1 1>]}}}))
    ScaleA=true
    ScaleRGB=true
   )
   /SetVariables(Particles.BackingSpriteSize:Vector2f=<3 2.5>) {isSetParameters=true, module=SetVariables_#2}
   &InheritSourceMovement(
    Applied Position Scale=&VectorFromFloat(Value=&FloatFromCurve(FloatCurve=DataInterface() {propertyValues {Curve {keys [<at 0 1 RCIM_Cubic>, {arriveTangent=-1, interpMode=RCIM_Cubic, leaveTangent=-1, time=0.8, value=0.85}, <at 1 0 RCIM_Cubic>]}}}))
    ApplyForce=false
    ApplyPosition=true
    ApplyPositionFalloff=false
   ) {enabled=false}
   &SolveForcesAndVelocity(Clamp Velocity=false, Limit Acceleration=false)
   &DynamicMaterialParameters(
    Index 0 Param 1=&Lerp_Float(
     A=1.25
     B=0.5
     Alpha=&FloatFromCurve(FloatCurve=DataInterface() {propertyValues {Curve {keys [{interpMode=RCIM_Cubic, tangentMode=RCTM_User, time=0, value=1}, {interpMode=RCIM_Cubic, tangentMode=RCTM_User, time=1, value=0}]}}})
    )
    Index 0 Param 2=&RandomRangeFloat(Minimum=-1, Maximum=1)
    Index 0 Param 3=&RandomRangeFloat(Minimum=-3, Maximum=3)
    Index 0 Param 4=20
    Module.Param0WriteEnabled=true
    Module.Param1WriteEnabled=true
    Module.Param2WriteEnabled=true
    Module.Param3WriteEnabled=true
   )
  ]
  rendererClasses [&NiagaraSpriteRendererProperties]
  renderers [
   {rendererClass=&NiagaraSpriteRendererProperties}
   {rendererClass=&NiagaraSpriteRendererProperties, rendererIndex=1}
  ]
 }
 Glow {
  emitterUpdateScript [
   &EmitterState()
   &SpawnRate(SpawnRate=25, Use Spawn Probability=false)
  ]
  particleSpawnScript [
   &InitializeParticle(
    Color=<rgba 15 10 1 1>
    Lifetime Max=0.5
    Lifetime Min=0.1
    Lifetime Mode=::NewEnumerator1
    Sprite Size Mode=::NewEnumerator1
    Uniform Sprite Size Max=1.75
    Uniform Sprite Size Min=1.25
    UsePositionOffset=false
    Write Material Random=true
   )
   &SphereLocation(
    MaskSpawnGroup=false
    Offset=&UniformRangedVector(Minimum=<0 0 -2>, Maximum=<0 0 -0.5>)
    Override Seed=false
    Sphere Radius=1.5
    Surface Only=false
    UseNonUniformScale=false
    UseOffset=true
    UseSphereOrientation=false
   )
   &Color(Color=&MultiplyLinearColor(
     Linear Color=$Particles.Color
     Scale Factor=/Niagara/DynamicInputs/UniformRange/UniformRangedVector4(
      Minimum:Vector4f=<1 0.25 0.25 0.01>
      Maximum:Vector4f=<1 1 1 1>
      Randomness Mode:ENiagaraRandomnessMode=::NewEnumerator0
      Random Seed:NiagaraInt32=0
      Fixed Random Seed:NiagaraBool=false
     )
    ))
  ]
  particleUpdateScript [
   &ParticleState()
   &SkeletalMeshLocation(
    Apply To Particle Position:NiagaraBool=&SetBoolByFloatComparison(A=$Particles.NormalizedAge, B=0.8, Comparison Type=::NewEnumerator6)
    Inherit Velocity=false
    Interpolate Mesh Sampling=false
    MaskSpawnGroup=false
    Mesh Sampling Type=::NewEnumerator5
    Sampled Position Offset Coordinate Space:ENiagaraMeshTransforms=::NewEnumerator1
    Skeletal Mesh=$User.SkeletalMesh
    Socket Sampling Mode=::NewEnumerator5
   ) {module=SkeletalMeshLocation001}
   &AddVelocity(Coordinate Space=::World, Velocity=&RandomRangeVector(Minimum=<0 0 0>, Maximum=<1 1 10>))
   &Drag(
    Drag=&VectorLengthSafe(Vector:Vector3f=$Output.SkeletalMeshLocation001.SampledVelocity)
    Use Linear Drag=true
    Use Rotational Drag=false
   )
   &SolveForcesAndVelocity(Clamp Velocity=false, Limit Acceleration=false)
   &ScaleColor(
    Scale Alpha=&FloatFromCurve(FloatCurve=DataInterface() {propertyValues {Curve {keys [<at 0.5 1>,<at 1 0>]}}})
    Scale RGB=&Lerp_Vector(A=<1 1 1>, B=<1 0 0>, Alpha=$Particles.NormalizedAge)
    ScaleA=true
    ScaleRGB=true
   )
   &KillParticlesInVolume(Invert Volume=true, Origin Offset:Vector3f=<0 0 3>, Sphere Radius=4, Volume Origin:Vector3f=$Output.SkeletalMeshLocation001.SampledPosition)
   &ScaleSpriteSize(Scale Factor=&Vector2DFromFloat(Value=&FloatFromCurve(FloatCurve=DataInterface() {propertyValues {Curve {keys [
          {arriveTangent=-0.0036, interpMode=RCIM_Cubic, leaveTangent=-0.0036, tangentMode=RCTM_User, time=0, value=1}
          {arriveTangent=-2.2201, interpMode=RCIM_Cubic, leaveTangent=-2.2201, tangentMode=RCTM_User, time=1, value=0}
         ]}}})))
   /SetVariables(
    Particles.ColorAlt:LinearColor=&MultiplyLinearColor(Linear Color=$Particles.Color, Scale Factor=<1 0.1 0.1 0.01>)
    Particles.SpriteSizeAlt:Vector2f=&Vector2DFromFloat(Value=&FloatFromCurve(
      FloatCurve=DataInterface() {propertyValues {Curve {keys [
          {arriveTangent=-0.0036, interpMode=RCIM_Cubic, leaveTangent=-0.0036, tangentMode=RCTM_User, time=0, value=1}
          {arriveTangent=-2.2201, interpMode=RCIM_Cubic, leaveTangent=-2.2201, tangentMode=RCTM_User, time=1, value=0}
         ]}}}
      Scale Curve=&RandomRangeFloat(Minimum=5, Maximum=10)
     ))
   ) {isSetParameters=true}
   &ScaleColor(
    Scale RGB=&VectorFromFloat(Value=&FN_FrontEndFloatModifier(Gameplay=1, Front End=0.15))
    ScaleA=false
    ScaleRGB=true
   ) {module=ScaleColor001}
  ]
  rendererClasses [&NiagaraSpriteRendererProperties]
  renderers [
   {rendererClass=&NiagaraSpriteRendererProperties}
   {rendererClass=&NiagaraSpriteRendererProperties, rendererIndex=1}
  ]
 }
 Light {
  emitterUpdateScript [
   &EmitterState(Scalability Mode=::NewEnumerator1)
   &SpawnRate(SpawnRate=6, Use Spawn Probability=false)
  ]
  particleSpawnScript [
   /SetVariables(Particles.SpriteSize=&Vector2DFromFloat(Value=5)) {isSetParameters=true}
   /SetVariables(Particles.Lifetime=&RandomRangeFloat(Minimum=0.2, Maximum=0.4)) {isSetParameters=true}
  ]
  particleUpdateScript [
   &ParticleState()
   &SkeletalMeshLocation(
    Inherit Velocity=false
    Interpolate Mesh Sampling=false
    MaskSpawnGroup=false
    Mesh Sampling Type=::NewEnumerator5
    Sampled Position Offset=<-6.3371 3.794 0>
    Sampled Position Offset Coordinate Space:ENiagaraMeshTransforms=::NewEnumerator3
    Skeletal Mesh=$User.SkeletalMesh
    Socket Sampling Mode=::NewEnumerator5
   ) {module=SkeletalMeshLocation001}
   &FN_Light_v3(
    Color=&RandomRangeLinearColor(Minimum=<rgba 7 0.35 0 1>, Maximum=<rgba 7 1.2833 0 1>)
    Create Color Attribute=true
    Create Exponent Attribute=true
    Create Radius Attribute=true
    Enable Particle Light Enable=true
    Exponent=2
    Light Enabled=true
    Radius:NiagaraFloat=&Multiply_Float(
     A=&RandomRangeFloat(Minimum=20, Maximum=26)
     B=&FloatFromCurve(FloatCurve=DataInterface() {propertyValues {Curve {keys [<at 0 0.5 RCIM_Cubic>, <at 0.25 1>, {arriveTangent=-1.3333, interpMode=RCIM_Cubic, leaveTangent=-1.3333, time=0.75, value=1}, <at 1 0>]}}})
    )
    Scale Color:NiagaraFloat=&FloatFromCurve(FloatCurve=DataInterface() {propertyValues {Curve {keys [
         {arriveTangent=5.0862, interpMode=RCIM_Cubic, leaveTangent=5.0862, tangentMode=RCTM_User, time=0, value=0}
         {interpMode=RCIM_Cubic, tangentMode=RCTM_User, time=0.5, value=1}
         {arriveTangent=-5.4921, interpMode=RCIM_Cubic, leaveTangent=-5.4922, tangentMode=RCTM_User, time=1, value=0}
        ]}}})
   )
  ]
  rendererClasses [&NiagaraSpriteRendererProperties, &NiagaraLightRendererProperties]
  renderers [
   {rendererClass=&NiagaraSpriteRendererProperties}
   {rendererClass=&NiagaraLightRendererProperties, rendererIndex=1}
  ]
 }
 Smoke {
  emitterUpdateScript [
   &EmitterState()
   &SpawnRate(SpawnRate=50, Use Spawn Probability=false)
  ]
  particleSpawnScript [
   &InitializeRibbon(
    Color=<rgba 10 1.75 0.2 0.3>
    Lifetime=0.6
    Module.Write Color=true
    Module.Write Lifetime=true
    Module.Write Mass=true
    Module.Write Material Random=true
    Module.Write Position=true
    Module.Write RibbonTwist=false
    Module.Write RibbonWidth=true
    Ribbon Width=&RandomRangeFloat(Minimum=1, Maximum=4)
   )
   /SetVariables(Particles.SpriteSize=&Vector2DFromFloat(Value=1)) {isSetParameters=true}
   &SkeletalMeshLocation(
    Inherit Velocity=true
    MaskSpawnGroup=false
    Mesh Sampling Type=::NewEnumerator5
    Sampled Position Offset Coordinate Space:ENiagaraMeshTransforms=::NewEnumerator3
    Skeletal Mesh=$User.SkeletalMesh
    Socket Sampling Mode=::NewEnumerator5
    VelocitySpeedLimit=false
   )
   /SetVariables(Particles.RibbonLinkOrder:NiagaraFloat=&MakeFloatFromInt(Int=&GenerateIntFromCounter(SimpleCounter:NiagaraDataInterfaceSimpleCounter=DataInterface() {propertyValues {GpuSyncMode=None, InitialValue=0}}))) {isSetParameters=true, module=SetVariables_#1}
  ]
  particleUpdateScript [
   &ParticleState()
   &SkeletalMeshLocation(
    Apply To Particle Position:NiagaraBool=&SetBoolByFloatComparison(A=$Particles.NormalizedAge, B=0.2, Comparison Type=::NewEnumerator6)
    Inherit Velocity=false
    Interpolate Mesh Sampling=false
    MaskSpawnGroup=false
    Mesh Sampling Type=::NewEnumerator5
    Sampled Position Offset Coordinate Space:ENiagaraMeshTransforms=::NewEnumerator3
    Skeletal Mesh=$User.SkeletalMesh
    Socket Sampling Mode=::NewEnumerator5
   ) {module=SkeletalMeshLocation001}
   /SetVariables(
    Particles.DistanceAbove:Vector3f=&self:ClampVector(
     Input Vector:Vector3f=&Multiply_Vector(
      A=<0 0 0.1>
      B=&Subtract_Vector(A=Linked() {linkedVariable {name=Particles.Position, type=NiagaraPosition}}, B=$Output.SkeletalMeshLocation001.SampledPosition)
     )
     Max:Vector3f=<0 0 2>
     Min:Vector3f=<0 0 0>
    )
    Particles.DistanceBelow:Vector3f=&self:ClampVector(
     Input Vector:Vector3f=&Multiply_Vector(
      A=<0 0 0.1>
      B=&Subtract_Vector(A=$Output.SkeletalMeshLocation001.SampledPosition, B=Linked() {linkedVariable {name=Particles.Position, type=NiagaraPosition}})
     )
     Max:Vector3f=<0 0 1>
     Min:Vector3f=<0 0 0>
    )
    Particles.DistanceFromBone:NiagaraFloat=&Multiply_Float(
     A=0.01
     B=&VectorLength(Vector=&Subtract_Vector(A=Linked() {linkedVariable {name=Particles.Position, type=NiagaraPosition}}, B=$Output.SkeletalMeshLocation001.SampledPosition))
    )
   ) {isSetParameters=true}
   /SetVariables(Particles.NormalizedAge:NiagaraFloat=&Lerp_Float(
     A=$Particles.NormalizedAge
     B=1
     Alpha=&ClampFloat(Float=&Add_Float(A=&MakeFloatFromVector(Vector=$Particles.DistanceAbove, Channel=::NewEnumerator2), B=-1.5))
    )) {isSetParameters=true}
   &GravityForce(Gravity=&Lerp_Vector(
     A=<0 0 700>
     B=<0 0 200>
     Alpha=&ClampFloat(Float=&Multiply_Float(A=&VectorLength(Vector=$Output.SkeletalMeshLocation001.SampledVelocity), B=0.5))
    ))
   &CurlNoiseForce(
    Noise Frequency=5
    Noise Strength=&FloatFromCurve(FloatCurve=DataInterface() {propertyValues {Curve {keys [{arriveTangent=957.2371, interpMode=RCIM_Cubic, leaveTangent=957.237, tangentMode=RCTM_User, time=0, value=10}, <at 1 300>]}}})
    Pan Noise=true
    Pan Noise Field=<0 0 2>
    Random Seed:NiagaraInt32=&RandomRangeInt(Minimum=0, Maximum=100)
   )
   &Drag(
    Drag=&Multiply_Float(
     A=&ClampFloat(Float=&VectorLengthSafe(Vector:Vector3f=$Particles.Velocity), Max=5000)
     B=&Lerp_Float(A=0.5, B=0, Alpha=$Particles.NormalizedAge)
    )
    Use Linear Drag=true
    Use Rotational Drag=false
   )
   &AddVelocity(Velocity=&Multiply_VectorByFloat(Vector=$Particles.DistanceBelow, Float=&RandomRangeFloat(Minimum=0, Maximum=100))) {module=AddVelocity003}
   &FN_RibbonMaterial(
    Color=&ColorFromCurve(ColorCurve=DataInterface() {propertyValues {
       AlphaCurve {keys [<at 0 1 RCIM_Cubic>]}
       BlueCurve {keys [<at 0 0.4902 RCIM_Cubic>,<at 0.3146 0.918 RCIM_Cubic>]}
       GreenCurve {keys [<at 0 1.3333 RCIM_Cubic>,<at 0.3146 0.8632 RCIM_Cubic>]}
       RedCurve {keys [<at 0 5 RCIM_Cubic>,<at 0.3146 0.8528 RCIM_Cubic>]}
      }})
    Color Pan U Speed=-2.5345
    Color Scale U=7.5
    Erosion Amount:NiagaraFloat=&Lerp_Float(A=0.5, B=1.6, Alpha=$Particles.NormalizedAge)
    Erosion Pan Speed U=-0.8346
    Erossion Scale U=1.5
    Opacity Pan U Speed=-1.5
    Opacity Power:NiagaraFloat=&RandomRangeFloat(Minimum=0.3, Maximum=1)
   )
   &ScaleRibbonWidth(Ribbon Width Scale=&FloatFromCurve(
     FloatCurve=DataInterface() {propertyValues {Curve {keys [
         {arriveTangent=4.0699, interpMode=RCIM_Cubic, leaveTangent=4.0699, tangentMode=RCTM_User, time=0, value=0}
         <at 0.5 0.9666 RCIM_Cubic>
         {arriveTangent=-2.5115, interpMode=RCIM_Cubic, leaveTangent=-2.5115, tangentMode=RCTM_User, time=1, value=0}
        ]}}}
     Scale Curve=&RandomRangeFloat(Minimum=0.5, Maximum=2)
    )) {module=ScaleRibbonWidth001}
   &SolveForcesAndVelocity(Clamp Velocity=false, Limit Acceleration=false)
   &Color(Color=&ColorFromCurve(
     ColorCurve=DataInterface() {propertyValues {
       AlphaCurve {keys [<at 0 1 RCIM_Cubic>]}
       BlueCurve {keys [<at 0 0 RCIM_Cubic>,<at 0.4 0>,<at 0.6 0>,<at 1 1 RCIM_Cubic>]}
       GreenCurve {keys [<at 0 0.0148 RCIM_Cubic>,<at 0.4 0.8005>,<at 0.6 1>,<at 1 0.087 RCIM_Cubic>]}
       RedCurve {keys [<at 0 1 RCIM_Cubic>,<at 0.4 1>,<at 0.6 0.0681>,<at 1 0 RCIM_Cubic>]}
      }}
     CurveIndex=$Particles.DistanceFromBone
    )) {enabled=false}
  ]
  rendererClasses [&NiagaraRibbonRendererProperties]
  renderers [{rendererClass=&NiagaraRibbonRendererProperties}]
 }
}
summary {
 systemName=NS_PickAxe_FNBirthday5_Idle
 emitters [
  {emitterName=Glow, rendererClasses [&NiagaraSpriteRendererProperties]}
  {emitterName=Flame, rendererClasses [&NiagaraRibbonRendererProperties]}
  {emitterName=Light, rendererClasses [&NiagaraSpriteRendererProperties, &NiagaraLightRendererProperties]}
  {emitterName=Smoke, rendererClasses [&NiagaraRibbonRendererProperties]}
  {emitterName=GlitchDetails, rendererClasses [&NiagaraSpriteRendererProperties]}
  {bEnabled=false, emitterName=GlitchLarge, rendererClasses [&NiagaraSpriteRendererProperties]}
  {bEnabled=false, emitterName=Glitches, rendererClasses [&NiagaraSpriteRendererProperties]}
  {bEnabled=false, emitterName=Fountain, rendererClasses [&NiagaraSpriteRendererProperties]}
 ]
 userVariables [
  User.bIsFrontEnd {type=NiagaraBool}
  User.bIsFrontEndPreview {type=NiagaraBool}
  User.bisSwinging {type=NiagaraBool}
  User.SkeletalMesh {
   type=NiagaraDataInterfaceSkeletalMesh
   defaultValue {
    struct=NiagaraExt_VariableValue_DataInterface
    value {dataInterface=&self:NiagaraDataInterfaceSkeletalMesh_0}
   }
  }
 ]
}
