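# Idle Niagara effect for the FN Birthday 8 pickaxe, adding candle flames, flickering blue-tinted light, smoke and drips at weapon sockets.
# FlamesAndGlows passes the delta between smoothed and current socket positions to its material, so flames react to motion.
# SmokeRibbon uses anim trails that shift color and fade by socket distance and time of day.
# Drips cling to the socket before falling, and stop spawning while bisSwinging is set.
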NiagaraSystem /BRCosmetics/Weapons/FORT_Melee/Pickaxe_FN_Birthday_Male/FX/NS_PickAxe_FNBirthday8_Idle
dependencies {
 usedDataInterfaces [
  &NiagaraDataInterfaceCurve
  &NiagaraDataInterfaceColorCurve
  &NiagaraDataInterfaceVectorCurve
  &NiagaraDataInterfaceVector2DCurve
  &NiagaraDataInterfaceSocketReader
 ]
 usedDynamicInputs []
 usedModules [
  &self:SystemSpawnScript.SystemScriptSource.SystemScriptGraph.NiagaraNodeAssignment_1.SetVariables_#1
  &self:SystemSpawnScript.SystemScriptSource.SystemScriptGraph.NiagaraNodeAssignment_0.SetVariables_#2
  &SystemState
 ]
}
emitters {
 CandleLight {
  emitterUpdateScript [
   &EmitterState(
    Inactive Response=::NewEnumerator1
    Life Cycle Mode=::NewEnumerator1
    Loop Behavior=::NewEnumerator1
    Loop Duration Mode=::NewEnumerator1
    UseLoopDelay=false
   )
   &SpawnBurst_Instantaneous(Use Spawn Probability=false, UseLoopCountLimit=false)
  ]
  particleSpawnScript [&InitializeParticle(
    Color=<rgba 0 0.553 1 1>
    Lifetime=$Emitter.CurrentLoopDuration
    Sprite Size Mode=::NewEnumerator3
    Uniform Sprite Size=50
    UsePositionOffset=false
    Write Material Random=true
   )]
  particleUpdateScript [
   &ParticleState(Kill Particles When Lifetime Has Elapsed=false, Let Infinitely Lived Particles Die When Emitter Deactivates=true)
   &SolveForcesAndVelocity(Clamp Velocity=false, Limit Acceleration=false)
   &SocketLocation(
    Position Sampling=::NewEnumerator1
    SamplePosOffset=true
    Socket Reader=$System.Socket Reader
    Socket Sampling Mode=::NewEnumerator3
    VelocitySpeedLimit=false
   )
   &ScaleColor(
    Scale Alpha=&Waveform(
     Waveform Count=::NewEnumerator1
     "Waveform [1]"=::NewEnumerator12
     "[1] Frequency"=1
     "[1] Amplitude Min/Max"=<0.2 1>
     "[2] Average"=false
     "Waveform [2]"=::NewEnumerator12
     "[2] Input Position"=0
     "[2] Phase"=$Emitter.Age
     "[2] Phase Scale"=1
     "[2] Compound Frequency"=2
     "[2] Frequency"=3
     "[2] Amplitude Min/Max"=<0.5 1.5>
     "[2] Scale"=1
     "[2] Bias"=0
     "[2] Exponent"=1
     "[2] Pulse Duration"=0.1
     "[2] Clip Triangle Min/Max"=<0 0>
     "[2] Random Seed"=0
    )
    Scale RGB=&MakeVectorFromLinearColorRGB(Color=&ColorFromCurve(
      ColorCurve=DataInterface() {propertyValues {
        AlphaCurve {keys [<at 0 1 RCIM_Cubic>]}
        BlueCurve {keys [<at 0.0691 2 RCIM_Cubic>,<at 0.2839 2>,<at 0.4497 1.5>,<at 0.6974 1.7251>,<at 0.8603 2>]}
        GreenCurve {keys [<at 0.0691 0.5563 RCIM_Cubic>,<at 0.2839 0.1907>,<at 0.4497 1.1019>,<at 0.6974 0.8922>,<at 0.8603 0.5563>]}
        RedCurve {keys [<at 0.0691 0.2012 RCIM_Cubic>,<at 0.2839 0.1512>,<at 0.4497 0.1513>,<at 0.6974 0.0329>,<at 0.8603 0.2012>]}
       }}
      CurveIndex=&Waveform("Waveform [1]"=::NewEnumerator9, "[1] Frequency"=1, "[1] Amplitude Min/Max"=<0 1>)
     ))
    ScaleA=true
    ScaleRGB=true
   )
   &FN_Light_Attributes(
    Color=&ColorFromCurve(
     ColorCurve=DataInterface() {propertyValues {
       AlphaCurve {keys [<at 0 1 RCIM_Cubic>]}
       BlueCurve {keys [<at 0.0691 3 RCIM_Cubic>,<at 0.3539 2>,<at 0.4902 0.5>,<at 0.6974 1.7251>,<at 0.8591 3>]}
       GreenCurve {keys [<at 0.0691 1.6527 RCIM_Cubic>,<at 0.3539 0.5374>,<at 0.4902 0.1765>,<at 0.6974 0.8922>,<at 0.8591 1.6527>]}
       RedCurve {keys [<at 0.0691 0.2464 RCIM_Cubic>,<at 0.3539 0.2663>,<at 0.4902 0.0313>,<at 0.6974 0.0329>,<at 0.8591 0.2464>]}
      }}
     CurveIndex=&Waveform("Waveform [1]"=::NewEnumerator9, "[1] Frequency"=2, "[1] Amplitude Min/Max"=<0 1>)
    )
    Create Color Attribute=true
    Create Exponent Attribute=true
    Create Position Attribute=true
    Create Radius Attribute=true
    Exponent=2
    Position:NiagaraPosition=$Output.SocketLocation.SampledPosition
    Radius=&Multiply_Float(
     A=50
     B=&FloatFromCurve(
      FloatCurve=DataInterface() {propertyValues {Curve {keys [
          {arriveTangent=-0.2724, interpMode=RCIM_Cubic, leaveTangent=-0.2724, tangentMode=RCTM_User, time=0, value=0}
          {arriveTangent=0.0224, interpMode=RCIM_Cubic, leaveTangent=0.0224, tangentMode=RCTM_User, time=0.5, value=1}
         ]}}}
      CurveIndex=$Engine.System.Age
     )
    )
   )
   &KillParticles(Kill Particles=&SetBoolByFloatComparison(A=$Particles.DistanceTraveled, B=100))
  ]
  rendererClasses [&NiagaraLightRendererProperties, &NiagaraMeshRendererProperties]
  renderers [
   {rendererClass=&NiagaraLightRendererProperties, rendererIndex=0}
   {rendererClass=&NiagaraMeshRendererProperties, rendererIndex=1}
  ]
 }
 Drips {
  emitterSpawnScript [/SetVariables(Emitter.AgeMultiplier=1, Emitter.AgeOffset=-1.5) {isSetParameters=true}]
  emitterUpdateScript [
   &EmitterState(Life Cycle Mode=::NewEnumerator1, Loop Duration=1, UseLoopDelay=false)
   &SpawnRate(
    Spawn Probability=&MakeCustomFloatFromBool(Bool=$User.bisSwinging, True Float=0, False Float=0.5)
    SpawnRate=0.7
    Use Spawn Probability=true
   )
  ]
  particleSpawnScript [
   &InitializeParticle(
    Color=&LerpLinearColors(StartColor=<rgba 1 1 1 1>, EndColor=$System.TOD, LerpFactor=0.7)
    Lifetime Max=4.5
    Lifetime Min=4.5
    Lifetime Mode=::NewEnumerator1
    Sprite Size Mode=::NewEnumerator3
    Uniform Sprite Size=&RandomRangeFloat(Minimum=0.8, Maximum=1)
    UsePositionOffset=false
    Write Material Random=true
   )
   /SetVariables(Particles.SpriteAlignment=<0 0 -1>) {enabled=false, isSetParameters=true}
   &CameraOffset(Camera Offset Amount=2)
  ]
  particleUpdateScript [
   &ParticleState()
   &SocketLocation(
    Position Sampling=::NewEnumerator1
    SamplePosOffset=true
    Sampled Position Offset=&Lerp_Vector(
     A=<-1.2 0 -3>
     B=<-1.5 0 -2>
     Alpha=&FloatFromCurve(FloatCurve=DataInterface() {propertyValues {Curve {keys [<at 0.2 0 RCIM_Cubic>,<at 0.5 1 RCIM_Cubic>]}}})
    )
    Socket Reader=$System.Socket Reader
    Socket Sampling Mode=::NewEnumerator3
    VelocitySpeedLimit=false
   )
   &GravityForce(Gravity=&MakeVector(
     X=0
     Y=0
     Z=&FloatFromCurve(
      FloatCurve=DataInterface() {propertyValues {Curve {keys [<at 0 0>, {arriveTangent=-2463.0549, leaveTangent=-2463.0549, time=0.4, value=0}, <at 0.6 -700>]}}}
      CurveIndex=&Add_Float(A=$Particles.Age, B=$Emitter.AgeOffset)
      Scale Curve=0.8
     )
    ))
   &LinearForce(
    Coordinate Space:ENiagaraCoordinateSpace=::World
    Force=&Multiply_Vector(
     A=&MakeCustomVectorFromBool(
      Bool=&SetBoolByFloatComparison(
       A=&MakeFloatFromVector(Vector=$Engine.Owner.Velocity, Channel=::NewEnumerator2)
       B=-50
       Comparison Type=::NewEnumerator6
      )
      True Vector=$Engine.Owner.Velocity
      False Vector=&Multiply_Vector(A=$Engine.Owner.Velocity, B=<1 1 0>)
     )
     B=<-0.2 -0.2 1>
    )
   ) {enabled=false}
   &SolveForcesAndVelocity(Clamp Velocity=true, Limit Acceleration=false)
   &ScaleSpriteSize(
    Scale Factor=&MakeVector2D(
     X=&FloatFromCurve(FloatCurve=DataInterface() {propertyValues {Curve {keys [
          {arriveTangent=-0.1045, interpMode=RCIM_Cubic, leaveTangent=-0.1045, tangentMode=RCTM_User, time=0.1, value=3}
          {arriveTangent=-3.75, interpMode=RCIM_Cubic, leaveTangent=-3.75, time=0.3096, value=2.4089}
          {arriveTangent=-4.8896, interpMode=RCIM_Cubic, leaveTangent=-4.8896, time=0.5, value=1.5}
          {interpMode=RCIM_Cubic, leaveTangent=0.625, time=0.7, value=0.5}
         ]}}})
     Y=&FloatFromCurve(FloatCurve=DataInterface() {propertyValues {Curve {keys [
          <at 0 0 RCIM_Cubic>
          {arriveTangent=5.3333, interpMode=RCIM_Cubic, leaveTangent=5.3333, time=0.5, value=4}
          {arriveTangent=25.2206, interpMode=RCIM_Cubic, leaveTangent=25.2206, tangentMode=RCTM_User, time=0.75, value=4}
          <at 1 4 RCIM_Cubic>
         ]}}})
    )
    Scale Sprite Size Mode=::NewEnumerator2
   )
   /SetVariables(Particles.Position=&LerpPosition(
     A=$Output.SocketLocation.SampledPosition
     B=$Particles.Position
     Alpha=&FloatFromCurve(
      FloatCurve=DataInterface() {propertyValues {Curve {keys [<at 0.5 0>,<at 0.6 1>]}}}
      CurveIndex=&Add_Float(A=$Particles.Age, B=$Emitter.AgeOffset)
     )
    )) {isSetParameters=true}
   /SetVariables(Particles.SpriteAlignment=&Lerp_Vector(
     A=&TransformVector(Vector=<-1 0 0>, Destination Space=::Simulation)
     B=&Add_Vector(A=<0 0 -1>, B=&Multiply_Vector(A=$Particles.Velocity, B=<0.2 0.2 0.2>))
     Alpha=&FloatFromCurve(FloatCurve=DataInterface() {propertyValues {Curve {keys [<at 0.2 0 RCIM_Cubic>,<at 0.4 1 RCIM_Cubic>]}}})
    )) {isSetParameters=true}
   &ScaleColor(
    Scale Alpha=1.5
    Scale RGB=&MakeVectorFromLinearColorRGB(Color=&ColorFromCurve(ColorCurve=DataInterface() {propertyValues {
        AlphaCurve {keys [<at 0 1 RCIM_Cubic>]}
        BlueCurve {keys [{arriveTangent=0.7394, interpMode=RCIM_Cubic, time=0.2425, value=1.5}, <at 0.6356 0.5776 RCIM_Cubic>]}
        GreenCurve {keys [{arriveTangent=-1.04, interpMode=RCIM_Cubic, time=0.2425, value=1.0082}, <at 0.6356 0.5457 RCIM_Cubic>]}
        RedCurve {keys [{arriveTangent=-2.8211, interpMode=RCIM_Cubic, time=0.2425, value=0.4062}, <at 0.6356 0.4452 RCIM_Cubic>]}
       }}))
    ScaleA=true
    ScaleRGB=true
   )
  ]
  rendererClasses [&NiagaraSpriteRendererProperties]
  renderers [{rendererClass=&NiagaraSpriteRendererProperties}]
 }
 FlamesAndGlows {
  emitterUpdateScript [
   &EmitterState(
    Inactive Response=::NewEnumerator1
    Life Cycle Mode=::NewEnumerator1
    Loop Behavior=::NewEnumerator1
    Loop Duration Mode=::NewEnumerator1
    UseLoopDelay=false
   )
   &SpawnBurst_Instantaneous(Use Spawn Probability=false, UseLoopCountLimit=false)
  ]
  particleSpawnScript [
   &FN_SetSpawnGroup(Number of Spawn Groups:NiagaraInt32=$System.NumCandles)
   &InitializeParticle(
    Color=<rgba 0 17.626 25 1>
    Lifetime Max=0.5
    Lifetime Min=0.1
    Lifetime Mode=::NewEnumerator1
    Position=$System.CandlePosition
    Position Mode=::NewEnumerator1
    Sprite Size Mode=::NewEnumerator1
    Uniform Sprite Size Max=20
    Uniform Sprite Size Min=20
    UsePositionOffset=true
    Write Material Random=true
   )
   /SetVariables(Particles.LaggedPosition:Vector3f=&ConvertPositionToVector(Input Position=<0 0 0>, Conversion Operation:ENiagara_LWCConvertPosToVec=::NewEnumerator0)) {isSetParameters=true}
   /SetVariables(Particles.ColorAlt:LinearColor=<rgba 0 0.4407 1 0.35>) {isSetParameters=true}
  ]
  particleUpdateScript [
   &ParticleState(
    Kill Particles When Lifetime Has Elapsed=false
    Let Infinitely Lived Particles Die When Emitter Deactivates=true
    Loop Particles Lifetime=true
   )
   &SocketLocation(
    Orientation Sampling=::NewEnumerator1
    Position Sampling=::NewEnumerator1
    SamplePosOffset=true
    Sampled Position Offset=<0 0 -0.2>
    Socket Reader=$System.Socket Reader
    Socket Sampling Mode=::NewEnumerator3
    VelocitySpeedLimit=false
   )
   /SetVariables(Particles.Position=&SubtractVectorFromPosition(
     Position=$Output.SocketLocation.SampledPosition
     Vector=&TransformVector(Vector=<0 0 0>, Destination Space=::Simulation)
    )) {isSetParameters=true}
   &SolveForcesAndVelocity(Clamp Velocity=false, Limit Acceleration=false)
   &ScaleColor(
    Scale Alpha=&Waveform(
     Global Input Position=&RandomRangeFloat(Minimum=0, Maximum=1)
     Global Phase=$System.NormalizedLoopAge
     "Waveform [1]"=::NewEnumerator12
     "[1] Frequency"=1
     "[1] Amplitude Min/Max"=<0 1>
    )
    Scale RGB=&Lerp_Vector(A=<1 1 1>, B=<1 0.5 0.25>, Alpha=$Particles.NormalizedAge)
    ScaleA=true
    ScaleRGB=true
   )
   &ScaleSpriteSize()
   /SetVariables(
    Particles.ColorAlt:LinearColor=&MultiplyLinearColor(
     Linear Color=$Particles.Initial.ColorAlt
     Scale Factor=&MakeVector4_FromFloat(Float=&Waveform(
       "Waveform [1]"=::NewEnumerator10
       "[1] Input Position"=&ReturnNormalizedExecIndex(Normalized Index Scale=0.8)
       "[1] Frequency"=5
       "[1] Amplitude Min/Max"=<0.5 1>
       "[1] Scale"=&Lerp_Float(
        A=1
        B=0.75
        Alpha=&MakeCustomFloatFromBool(
         Bool=&SetBoolByFloatComparison(A=200, B=&VectorLength(Vector=$Engine.Owner.Velocity))
         True Float=0
         False Float=1
        )
       )
      ))
    )
    Particles.SpriteSizeAlt:Vector2f=&Vector2DFromFloat(Value=&FloatFromCurve(
      FloatCurve=DataInterface() {propertyValues {Curve {keys [
          {arriveTangent=-0.0036, interpMode=RCIM_Cubic, leaveTangent=-0.0036, tangentMode=RCTM_User, time=0, value=1}
          {arriveTangent=-2.2201, interpMode=RCIM_Cubic, leaveTangent=-2.2201, tangentMode=RCTM_User, time=1, value=0}
         ]}}}
      Scale Curve=&RandomRangeFloat(Minimum=5, Maximum=10)
     ))
   ) {isSetParameters=true}
   /SetVariables(Particles.MaterialRandom:NiagaraFloat=&RandomRangeFloat(Minimum=0, Maximum=1)) {isSetParameters=true}
   /SetVariables(Particles.LaggedPosition:Vector3f=/Niagara/DynamicInputs/Lerp/SmoothLerpOverTime_Position(
     Convergence Rate:NiagaraFloat=17
     Smooth Value:NiagaraPosition=Linked() {linkedVariable {name=Particles.LaggedPosition, type=Vector3f}}
     Target Value:NiagaraPosition=$Output.SocketLocation.SampledPosition
    )) {isSetParameters=true}
   /SetVariables(Particles.PositionDelta:Vector3f=&Subtract_Vector(A=$Particles.LaggedPosition, B=Linked() {linkedVariable {name=Output.SocketLocation.SampledPosition, type=NiagaraPosition}})) {isSetParameters=true}
   &ScaleColor(
    Scale RGB=&VectorFromFloat(Value=&FN_FrontEndFloatModifier(Gameplay=1, Front End=0.15))
    ScaleA=false
    ScaleRGB=true
   ) {module=ScaleColor001}
   &DynamicMaterialParameters(
    Index 0 Param 1=3
    Index 0 Param 2=&MakeFloatFromVector(Vector=$Particles.PositionDelta, Channel=::NewEnumerator0)
    Index 0 Param 3=&MakeFloatFromVector(Vector=$Particles.PositionDelta, Channel=::NewEnumerator1)
    Index 0 Param 4=&ClampFloat(Float=&MakeFloatFromVector(Vector=$Particles.PositionDelta, Channel=::NewEnumerator2), Max=10)
    Module.Index0Param0Write=true
    Module.Index0Param1Write=true
    Module.Index0Param2Write=true
    Module.Index0Param3Write=true
   )
   &CameraOffset(Camera Offset Amount=2)
  ]
  rendererClasses [&NiagaraSpriteRendererProperties, &NiagaraMeshRendererProperties]
  renderers [
   {rendererClass=&NiagaraSpriteRendererProperties}
   {rendererClass=&NiagaraMeshRendererProperties, rendererIndex=1}
  ]
 }
 SmokeRibbon {
  emitterUpdateScript [
   &EmitterState()
   &FN_SpawnAnimTrail(Spawn Interval=2)
  ]
  particleSpawnScript [
   &FN_FlipRibbonV()
   &FN_InitializeAnimTrailBoneIndecies(Second Bone or Socket Index=0)
   &FN_InitializeAnimTrail(AnimTrailWidth=5, Color=<rgba 0.3802 0.3802 0.3802 0.35>, Lifetime=0.45, Position Offset=<0 0 2>)
   &AddVelocity(
    Coordinate Space=::World
    Velocity=&MakeVector(X=0, Y=0, Z=&Waveform("Waveform [1]"=::NewEnumerator9, "[1] Frequency"=3, "[1] Amplitude Min/Max"=<50 100>))
   )
   &AddVelocity(
    Rotation Coordinate Space=::World
    Velocity=&MakeVector(
     X=&Waveform(
      Global Input Position=&RandomRangeFloat(Minimum=0, Maximum=100, Randomness Mode=::NewEnumerator1, Random Seed=2, Fixed Random Seed=true)
      "Waveform [1]"=::NewEnumerator12
      "[1] Frequency"=4
      "[1] Amplitude Min/Max"=<-1 1>
      "[1] Scale"=&RandomRangeFloat(Minimum=15, Maximum=30)
     )
     Y=&Waveform(
      Global Input Position=&RandomRangeFloat(Minimum=0, Maximum=100, Randomness Mode=::NewEnumerator1, Random Seed=5, Fixed Random Seed=true)
      "Waveform [1]"=::NewEnumerator12
      "[1] Frequency"=6
      "[1] Amplitude Min/Max"=<-1 1>
      "[1] Scale"=&RandomRangeFloat(Minimum=15, Maximum=30)
     )
     Z=0
    )
   ) {module=AddVelocity001}
  ]
  particleUpdateScript [
   &ScaleVelocity(
    Coordinate Space:ENiagaraCoordinateSpace=::World
    Velocity Scale=&VectorFromCurve(VectorCurve=DataInterface() {propertyValues {
       XCurve {keys [
         <at 0 0 RCIM_Cubic>
         {arriveTangent=1.6667, interpMode=RCIM_Cubic, leaveTangent=1.6667, time=0.3, value=1}
         {arriveTangent=-1.8983, interpMode=RCIM_Cubic, leaveTangent=-1.8983, time=0.6, value=1}
         <at 0.7214 0.2 RCIM_Cubic>
         <at 0.8619 1 RCIM_Cubic>
        ]}
       YCurve {keys [
         <at 0 0 RCIM_Cubic>
         {arriveTangent=1.6667, interpMode=RCIM_Cubic, leaveTangent=1.6667, time=0.3, value=1}
         {arriveTangent=-1.8983, interpMode=RCIM_Cubic, leaveTangent=-1.8983, time=0.6, value=1}
         <at 0.7214 0.2 RCIM_Cubic>
         <at 0.8619 1 RCIM_Cubic>
        ]}
       ZCurve {keys [<at 0 1>,<at 1 1>]}
      }})
   )
   &SocketLocation(
    Position Sampling=::NewEnumerator1
    SamplePosOffset=true
    Socket Reader=$System.Socket Reader
    Socket Sampling Mode=::NewEnumerator3
    VelocitySpeedLimit=false
   )
   &SolveForcesAndVelocity(Clamp Velocity=false, Limit Acceleration=false)
   &FN_UpdateAnimTrail()
   &FN_RibbonMaterial()
   /SetVariables(Particles.RibbonDistance:NiagaraFloat=&DistanceBetweenPositions(A=$Particles.Position, B=$Output.SocketLocation.SampledPosition)) {isSetParameters=true}
   &ScaleRibbonWidth(Ribbon Width Scale=&FloatFromCurve(FloatCurve=DataInterface() {propertyValues {Curve {keys [{arriveTangent=4.5879, interpMode=RCIM_Cubic, time=0, value=0.3}, {interpMode=RCIM_Cubic, tangentMode=RCTM_User, time=1, value=1.6}]}}}))
   &ScaleColor(
    Scale Alpha=&FloatFromCurve(
     FloatCurve=DataInterface() {propertyValues {Curve {keys [{arriveTangent=-1.9717, interpMode=RCIM_Cubic, time=0, value=1}, <at 1 0 RCIM_Cubic>]}}}
     Scale Curve=&Multiply_Float(
      A=&FNTimeOfDayScalar(Night Scale=0.5, Morning Scale=1, Evening Scale=1)
      B=&FloatFromCurve(FloatCurve=DataInterface() {propertyValues {Curve {keys [<at 30 1 RCIM_Cubic>,<at 50 0 RCIM_Cubic>]}}}, CurveIndex=$Particles.RibbonDistance)
     )
    )
    Scale RGB=&MakeVectorFromLinearColorRGB(Color=&ColorFromCurve(
      ColorCurve=DataInterface() {propertyValues {
        AlphaCurve {keys [<at 0 1 RCIM_Cubic>]}
        BlueCurve {keys [<at 0.0214 1 RCIM_Cubic>,<at 0.4222 1 RCIM_Cubic>]}
        GreenCurve {keys [<at 0.0214 0.5272 RCIM_Cubic>,<at 0.4222 1 RCIM_Cubic>]}
        RedCurve {keys [<at 0.0214 0 RCIM_Cubic>,<at 0.4222 1 RCIM_Cubic>]}
       }}
      CurveIndex=&FloatFromCurve(FloatCurve=DataInterface() {propertyValues {Curve {keys [<at 3 0>,<at 15 1>]}}}, CurveIndex=$Particles.RibbonDistance)
     ))
    ScaleA=true
    ScaleRGB=true
   )
   &KillParticles(Kill Particles=&SetBoolByFloatComparison(A=$Particles.RibbonDistance, B=80))
  ]
  rendererClasses [&NiagaraRibbonRendererProperties]
  renderers [{rendererClass=&NiagaraRibbonRendererProperties}]
 }
}
summary {
 systemName=NS_PickAxe_FNBirthday8_Idle
 emitters [
  {emitterName=FlamesAndGlows, rendererClasses [&NiagaraSpriteRendererProperties, &NiagaraMeshRendererProperties]}
  {emitterName=CandleLight, rendererClasses [&NiagaraLightRendererProperties, &NiagaraMeshRendererProperties]}
  {emitterName=SmokeRibbon, rendererClasses [&NiagaraRibbonRendererProperties]}
  {emitterName=Drips, rendererClasses [&NiagaraSpriteRendererProperties]}
 ]
 userVariables [
  User.bIsFrontEnd {type=NiagaraBool}
  User.bIsFrontEndPreview {type=NiagaraBool}
  User.bisSwinging {type=NiagaraBool}
  User.RibbonIDOffset {type=NiagaraInt32}
 ]
}
