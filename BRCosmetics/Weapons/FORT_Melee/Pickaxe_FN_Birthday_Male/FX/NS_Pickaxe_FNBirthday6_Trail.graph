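# Niagara swing-trail effect for the sixth Fortnite birthday pickaxe, throwing cake bits, drops and colored sprinkles behind skeletal-mesh-driven ribbon trails.
# Cake, Drops and Sprinkles mesh emitters spawn per unit moved in a box around a sampled socket, then fall under gravity while shrinking, with speed-based stretch.
# Two AnimTrail ribbon emitters of differing width follow User.SkeletalMesh, timed by system delay and length variables.
# Two older FrostingTrail ribbon emitters are disabled.
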NiagaraSystem /BRCosmetics/Weapons/FORT_Melee/Pickaxe_FN_Birthday_Male/FX/NS_Pickaxe_FNBirthday6_Trail
dependencies {
 usedDataInterfaces [&NiagaraDataInterfaceCurve, &NiagaraDataInterfaceColorCurve]
 usedDynamicInputs []
 usedModules [
  &self:SystemSpawnScript.SystemScriptSource.SystemScriptGraph.NiagaraNodeAssignment_0.SetVariables_#1
  &SystemState
 ]
}
emitters {
 Cake {
  emitterSpawnScript [&FN_PreserveSkydiveUserVariables() {module=PreserveUserVariables}]
  emitterUpdateScript [
   &EmitterState(Life Cycle Mode=::NewEnumerator1, Loop Behavior=::NewEnumerator1, Loop Delay=0.03, Loop Duration=0.3, UseLoopDelay=true)
   &SpawnPerUnit(Module.Use Max Movement Threshold=true, Module.Use Movement Tolerance=false, Spawn Per Unit=75, Use Spawn Probability=false)
  ]
  particleSpawnScript [
   &FN_SetSpawnGroup(Number of Spawn Groups=2)
   &InitializeParticle(
    Color Maximum=<rgba 0.3854 0 0.0957 1>
    Color Minimum=<rgba 1 1 1 1>
    Color Mode=::NewEnumerator2
    Lifetime Max=1
    Lifetime Min=0.5
    Lifetime Mode=::NewEnumerator1
    Mesh Scale Mode=::NewEnumerator3
    Mesh Uniform Scale=&Lerp_Float(
     A=0.01
     B=0.02
     Alpha=&FloatFromCurve(
      FloatCurve=DataInterface() {propertyValues {Curve {keys [
          {interpMode=RCIM_Cubic, tangentMode=RCTM_Break, time=0, value=0}
          {arriveTangent=2.5592, interpMode=RCIM_Cubic, tangentMode=RCTM_Break, time=1, value=1}
         ]}}}
      CurveIndex=&RandomRangeFloat(Minimum=0, Maximum=1)
     )
    )
    UsePositionOffset=false
    Write Material Random=true
   )
   &SkeletalMeshLocation(
    Bone / Socket Index=1
    MaskSpawnGroup=false
    Mesh Sampling Type=::NewEnumerator5
    Skeletal Mesh=$User.SkeletalMesh
    Socket Sampling Mode=::NewEnumerator5
   )
   &BoxLocation(
    Box Origin:NiagaraPosition=$Output.SkeletalMeshLocation.SampledPositionWithoutOffset
    Box Size=<7 1 25>
    MaskSpawnGroup=false
    Module.Override Seed=false
    Rotate Box=false
    Surface Only=false
   )
   &AddVelocity(Velocity=&UniformRangedVector(Minimum=<0 -2 0>, Maximum=<0 5 0>))
   &AddVelocityInCone(
    Cone Angle=75
    Module.Use Velocity Falloff On Cone Axis=false
    Velocity Strength=&UniformRangedFloat(Minimum=1, Maximum=200)
   )
   &InheritVelocity(Module.Limit Speed=true, Velocity Scale=<1 1 0.5>) {module=InheritVelocity001}
   &AddVelocityFromPoint(
    Module.Use Falloff Distance=false
    Velocity Origin:Vector3f=&SimulationPosition()
    Velocity Strength=&UniformRangedFloat(Minimum=10, Maximum=100)
   )
   /SetVariables(Particles.MeshScaleRandom:NiagaraFloat=&UniformRangedFloat(Minimum=1, Maximum=2)) {isSetParameters=true, module=SetVariables_#2}
   &InitialMeshOrientation(Use Rotation Vector=false)
  ]
  particleUpdateScript [
   &ParticleState()
   &GravityForce(Gravity=<0 0 -700>)
   &SolveForcesAndVelocity(Clamp Velocity=false, Limit Acceleration=false)
   &ScaleMeshSize(Scale Factor=&VectorFromFloat(Value=&FloatFromCurve(
      FloatCurve=DataInterface() {propertyValues {Curve {keys [<at 0 1 RCIM_Cubic>,<at 0.2048 0.8914>,<at 1 0 RCIM_Cubic>]}}}
      Scale Curve=&Multiply_Float(A=$Particles.MeshScaleRandom, B=1.2)
     )))
   &ScaleColor(
    Scale Alpha=&FloatFromCurve(FloatCurve=DataInterface() {propertyValues {Curve {keys [
         {interpMode=RCIM_Cubic, tangentMode=RCTM_User, time=0, value=1}
         {arriveTangent=-0.3941, interpMode=RCIM_Cubic, leaveTangent=-0.3941, tangentMode=RCTM_User, time=1, value=0}
        ]}}})
    Scale RGB=&VectorFromFloat(Value=&FloatFromCurve(FloatCurve=DataInterface() {propertyValues {Curve {keys [
          {interpMode=RCIM_Cubic, tangentMode=RCTM_Break, time=0.6, value=1}
          {arriveTangent=-0.9546, interpMode=RCIM_Cubic, tangentMode=RCTM_Break, time=1, value=0}
         ]}}}))
    ScaleA=false
    ScaleRGB=true
   )
   &ScaleMeshSizeBySpeed(
    Max Scale Factor=<2 1.5 1>
    Speed Threshold:NiagaraFloat=&FloatFromCurve(FloatCurve=DataInterface() {propertyValues {Curve {keys [
         {arriveTangent=2294.4238, interpMode=RCIM_Cubic, leaveTangent=2294.4238, tangentMode=RCTM_User, time=0, value=150}
         {arriveTangent=78.1625, interpMode=RCIM_Cubic, leaveTangent=78.3263, tangentMode=RCTM_User, time=1, value=2000}
        ]}}})
   )
   &UpdateMeshOrientation(
    Rotation Rate=&RandomRangeFloat(Minimum=-2, Maximum=2)
    Rotation Vector=&RandomRangeVector(Minimum=<0 0 0>, Maximum=<1 1 1>)
   )
  ]
  rendererClasses [&NiagaraMeshRendererProperties]
  renderers [{rendererClass=&NiagaraMeshRendererProperties, rendererIndex=0}]
 }
 Drops {
  emitterSpawnScript [&FN_PreserveSkydiveUserVariables() {module=PreserveUserVariables}]
  emitterUpdateScript [
   &EmitterState(Life Cycle Mode=::NewEnumerator1, Loop Behavior=::NewEnumerator1, Loop Delay=0.03, Loop Duration=0.3, UseLoopDelay=true)
   &SpawnPerUnit(Module.Use Max Movement Threshold=true, Module.Use Movement Tolerance=false, Spawn Per Unit=75, Use Spawn Probability=false)
  ]
  particleSpawnScript [
   &FN_SetSpawnGroup(Number of Spawn Groups=2)
   /Niagara/Modules/Spawn/Initialization/InitializeParticle(
    Lifetime=&UniformRangedFloat(Minimum=0.5, Maximum=1)
    Module.Write Color=true
    Module.Write Lifetime=true
    Module.Write Mass=true
    Module.Write Material Random=true
    Module.Write Position=false
    Module.Write Scale=false
    Module.Write SpriteRotation=false
    Module.Write SpriteSize=false
    Position=&SimulationPosition()
   )
   &SkeletalMeshLocation(
    Bone / Socket Index=1
    MaskSpawnGroup=false
    Mesh Sampling Type=::NewEnumerator5
    Skeletal Mesh=$User.SkeletalMesh
    Socket Sampling Mode=::NewEnumerator5
   )
   &BoxLocation(
    Box Origin:NiagaraPosition=$Output.SkeletalMeshLocation.SampledPositionWithoutOffset
    Box Size=<7 1 25>
    MaskSpawnGroup=false
    Module.Override Seed=false
    Rotate Box=false
    Surface Only=false
   )
   &AddVelocity(Velocity=&UniformRangedVector(Minimum=<0 -2 0>, Maximum=<0 5 0>))
   &AddVelocityInCone(
    Cone Angle=75
    Module.Use Velocity Falloff On Cone Axis=false
    Velocity Strength=&UniformRangedFloat(Minimum=1, Maximum=200)
   )
   &InheritVelocity(Module.Limit Speed=true, Velocity Scale=<1 1 0.5>) {module=InheritVelocity001}
   &AddVelocityFromPoint(
    Module.Use Falloff Distance=false
    Velocity Origin:Vector3f=&SimulationPosition()
    Velocity Strength=&UniformRangedFloat(Minimum=10, Maximum=100)
   )
   /SetVariables(Particles.MeshScaleRandom:NiagaraFloat=&UniformRangedFloat(Minimum=1, Maximum=2)) {isSetParameters=true, module=SetVariables_#2}
  ]
  particleUpdateScript [
   &ParticleState()
   &GravityForce(Gravity=<0 0 -900>)
   &SolveForcesAndVelocity(Clamp Velocity=false, Limit Acceleration=false)
   &ScaleMeshSize(Scale Factor=&VectorFromFloat(Value=&FloatFromCurve(
      FloatCurve=DataInterface() {propertyValues {Curve {keys [<at 0 1 RCIM_Cubic>,<at 0.2048 0.8914>,<at 1 0 RCIM_Cubic>]}}}
      Scale Curve=&Multiply_Float(A=$Particles.MeshScaleRandom, B=1.2)
     )))
   &ScaleColor(
    Scale Alpha=&FloatFromCurve(FloatCurve=DataInterface() {propertyValues {Curve {keys [
         {interpMode=RCIM_Cubic, tangentMode=RCTM_User, time=0, value=1}
         {arriveTangent=-0.3941, interpMode=RCIM_Cubic, leaveTangent=-0.3941, tangentMode=RCTM_User, time=1, value=0}
        ]}}})
    Scale RGB=&VectorFromFloat(Value=&FloatFromCurve(FloatCurve=DataInterface() {propertyValues {Curve {keys [
          {interpMode=RCIM_Cubic, tangentMode=RCTM_Break, time=0.6, value=1}
          {arriveTangent=-0.9546, interpMode=RCIM_Cubic, tangentMode=RCTM_Break, time=1, value=0}
         ]}}}))
    ScaleA=false
    ScaleRGB=true
   )
   &ScaleMeshSizeBySpeed(
    Max Scale Factor=<2 1.5 1>
    Speed Threshold:NiagaraFloat=&FloatFromCurve(FloatCurve=DataInterface() {propertyValues {Curve {keys [
         {arriveTangent=2294.4238, interpMode=RCIM_Cubic, leaveTangent=2294.4238, tangentMode=RCTM_User, time=0, value=150}
         {arriveTangent=78.1625, interpMode=RCIM_Cubic, leaveTangent=78.3263, tangentMode=RCTM_User, time=1, value=2000}
        ]}}})
   )
  ]
  rendererClasses [&NiagaraMeshRendererProperties]
  renderers [{rendererClass=&NiagaraMeshRendererProperties, rendererIndex=0}]
 }
 FN_AnimTrail {
  emitterSpawnScript [&FN_RibbonIDOffsetUtility()]
  emitterUpdateScript [
   &EmitterState(Life Cycle Mode=::NewEnumerator1, Loop Behavior=::NewEnumerator1, Loop Delay=$System.TrailDelay, Loop Duration=0.2, UseLoopDelay=true)
   &FN_SpawnAnimTrail(Spawn Mode:ENiagaraSkeletalSpawning=::NewEnumerator0)
  ]
  particleSpawnScript [
   &FN_InitializeAnimTrailBoneIndecies(Second Bone or Socket Index=0)
   &FN_InitializeAnimTrailBoneIndecies(AnimTrail Selection=1, First Bone or Socket Index=1) {enabled=false, module=FN_InitializeAnimTrailBoneIndecies001}
   &FN_InitializeAnimTrail(Lifetime=$System.TrailLength)
   &AddVelocity(Velocity=<0 400 0>) {enabled=false}
  ]
  particleUpdateScript [
   &SolveForcesAndVelocity(Clamp Velocity=false, Limit Acceleration=false)
   &FN_UpdateAnimTrail(Locked Particle RibbonWidth=1, SkeletalMesh=$User.SkeletalMesh) {module=FN_UpdateAnimTrail001}
   /SetVariables(Particles.VisibilityTag=$Particles.SpawnGroup) {enabled=false, isSetParameters=true, module=SetVariables_#3}
   &ScaleColor(
    Scale Alpha=&FloatFromCurve(
     FloatCurve=DataInterface() {propertyValues {Curve {keys [
         {interpMode=RCIM_Cubic, tangentMode=RCTM_User, time=0, value=1}
         {arriveTangent=-2.0069, interpMode=RCIM_Cubic, leaveTangent=-2.0069, tangentMode=RCTM_User, time=0.9999, value=0}
        ]}}}
     Scale Curve=0.95
    )
    Scale RGB=<5 5 5>
    ScaleA=true
    ScaleRGB=false
   )
   /SetVariables(Particles.RibbonWidth=35) {isSetParameters=true, module=SetVariables_#4}
  ]
  rendererClasses [&NiagaraRibbonRendererProperties]
  renderers [
   {rendererClass=&NiagaraRibbonRendererProperties}
   {rendererClass=&NiagaraRibbonRendererProperties, rendererIndex=1}
   {rendererClass=&NiagaraRibbonRendererProperties, rendererIndex=2}
  ]
 }
 FN_AnimTrail001 {
  emitterSpawnScript [&FN_RibbonIDOffsetUtility()]
  emitterUpdateScript [
   &EmitterState(Life Cycle Mode=::NewEnumerator1, Loop Behavior=::NewEnumerator1, Loop Delay=$System.TrailDelay, Loop Duration=0.2, UseLoopDelay=true)
   &FN_SpawnAnimTrail(Spawn Mode:ENiagaraSkeletalSpawning=::NewEnumerator0)
  ]
  particleSpawnScript [
   &FN_InitializeAnimTrailBoneIndecies(First Bone or Socket Index=1)
   &FN_InitializeAnimTrailBoneIndecies(AnimTrail Selection=1, First Bone or Socket Index=1) {enabled=false, module=FN_InitializeAnimTrailBoneIndecies001}
   &FN_InitializeAnimTrail(Lifetime=$System.TrailLength)
   &AddVelocity(Velocity=<0 400 0>) {enabled=false}
  ]
  particleUpdateScript [
   &SolveForcesAndVelocity(Clamp Velocity=false, Limit Acceleration=false)
   &FN_UpdateAnimTrail(Locked Particle RibbonWidth=1, SkeletalMesh=$User.SkeletalMesh) {module=FN_UpdateAnimTrail001}
   /SetVariables(Particles.VisibilityTag=$Particles.SpawnGroup) {enabled=false, isSetParameters=true, module=SetVariables_#3}
   &ScaleColor(
    Scale Alpha=&FloatFromCurve(
     FloatCurve=DataInterface() {propertyValues {Curve {keys [
         {interpMode=RCIM_Cubic, tangentMode=RCTM_User, time=0, value=1}
         {arriveTangent=-2.0069, interpMode=RCIM_Cubic, leaveTangent=-2.0069, tangentMode=RCTM_User, time=0.9999, value=0}
        ]}}}
     Scale Curve=0.95
    )
    Scale RGB=<5 5 5>
    ScaleA=true
    ScaleRGB=false
   )
   /SetVariables(Particles.RibbonWidth=50) {isSetParameters=true, module=SetVariables_#4}
   &DynamicMaterialParameters(
    Index 0 Param 1=0.288
    Index 0 Param 2=&FloatFromCurve(FloatCurve=DataInterface() {propertyValues {Curve {keys [<at 0 1>,<at 1 0>]}}})
    Module.Index0Param0Write=true
    Module.Index0Param1Write=true
    Module.Index0Param2Write=true
    Module.Index0Param3Write=true
   )
  ]
  rendererClasses [&NiagaraRibbonRendererProperties]
  renderers [
   {rendererClass=&NiagaraRibbonRendererProperties}
   {rendererClass=&NiagaraRibbonRendererProperties, rendererIndex=1}
   {rendererClass=&NiagaraRibbonRendererProperties, rendererIndex=2}
  ]
 }
 FrostingTrail {
  bEnabled=false
  emitterUpdateScript [
   &EmitterState(Life Cycle Mode=::NewEnumerator1, Loop Behavior=::NewEnumerator1, Loop Duration=0.3, UseLoopDelay=true)
   /Game/Effects/Niagara/Modules/Ribbons/FN_SpawnRibbon_v3(Skeleton Mesh:NiagaraDataInterfaceSkeletalMesh=$User.SkeletalMesh, Spawn Mode:ENiagaraSkeletalSpawning=::NewEnumerator1, SpawnRate=50)
  ]
  particleSpawnScript [
   /Game/Effects/Niagara/Modules/Ribbons/FN_InitializeRibbon_v5(
    Color:LinearColor=<rgba 0 1 0 1>
    Enable Skeletal Position=true
    Facing Axis:ENiagaraOrientationAxis=::YAxis
    Lifetime=0.08
    Offset Amount=-20
    Offset Axis:ENiagaraOrientationAxis=::ZAxis
    Ribbon Width=35
    SkeletalMesh:NiagaraDataInterfaceSkeletalMesh=$User.SkeletalMesh
   )
   &self:ColorFromTOD(
    Afternoon:LinearColor=<rgba 0.5122 0.5122 0.5122 1>
    Morning:LinearColor=<rgba 0.4948 0.4948 0.4948 1>
    Night:LinearColor=<rgba 0.2524 0.3428 0.5365 1>
    TOD:Vector4f=&MakeVector4_FromLinearColor(LinearColor=$NPC.FortniteNPC.FortniteActiveTimeOfDay)
   ) {enabled=false, module=ColorFromTOD}
  ]
  particleUpdateScript [
   /Game/Effects/Niagara/Modules/Ribbons/FN_UpdateRibbon_v5(Enable Skeletal Position=true, Interpolate Bone Sampling=true, SkeletalMesh:NiagaraDataInterfaceSkeletalMesh=$User.SkeletalMesh)
   /SetVariables(Particles.RibbonFacing=<1 0 0>) {isSetParameters=true, module=SetVariables_#5}
   &ScaleColor(
    Scale Alpha=&FloatFromCurve(FloatCurve=DataInterface() {propertyValues {Curve {keys [<at 0 0>,<at 0.1 1>,<at 0.2 1>,<at 1 0.4>]}}}, Scale Curve=0.95)
    Scale RGB=<5 5 5>
    ScaleA=true
    ScaleRGB=false
   )
   &FN_RibbonMaterial()
  ]
  rendererClasses [&NiagaraRibbonRendererProperties, &NiagaraMeshRendererProperties]
  renderers [
   {rendererClass=&NiagaraRibbonRendererProperties}
   {rendererClass=&NiagaraRibbonRendererProperties, rendererIndex=1}
   {rendererClass=&NiagaraMeshRendererProperties, rendererIndex=2}
  ]
 }
 FrostingTrail001 {
  bEnabled=false
  emitterUpdateScript [
   &EmitterState(Life Cycle Mode=::NewEnumerator1, Loop Behavior=::NewEnumerator1, Loop Duration=0.3, UseLoopDelay=true)
   &FN_SpawnRibbon_v3(Skeleton Mesh:NiagaraDataInterfaceSkeletalMesh=$User.SkeletalMesh, Spawn Mode:ENiagaraSkeletalSpawning=::NewEnumerator1, SpawnRate=50)
  ]
  particleSpawnScript [
   &FN_InitializeRibbon_v5(
    Color:LinearColor=<rgba 1 0 0 1>
    Enable Skeletal Position=true
    Facing Axis:ENiagaraOrientationAxis=::YAxis
    Interpolate Bone Sampling=false
    Lifetime=0.06
    Offset Amount=-35
    Offset Axis:ENiagaraOrientationAxis=::ZAxis
    Ribbon Width=50
    SkeletalMesh:NiagaraDataInterfaceSkeletalMesh=$User.SkeletalMesh
   )
   &self:ColorFromTOD(
    Morning:LinearColor=<rgba 0.4948 0.4948 0.4948 1>
    Night:LinearColor=<rgba 0.2524 0.3428 0.5365 1>
    TOD:Vector4f=&MakeVector4_FromLinearColor(LinearColor=$NPC.FortniteNPC.FortniteActiveTimeOfDay)
   ) {enabled=false, module=ColorFromTOD}
  ]
  particleUpdateScript [
   &FN_UpdateRibbon_v5()
   /SetVariables(Particles.RibbonFacing=<1 0 0>) {isSetParameters=true, module=SetVariables_#5}
   &ScaleColor(
    Scale Alpha=&FloatFromCurve(FloatCurve=DataInterface() {propertyValues {Curve {keys [<at 0 0>,<at 0.1 1>,<at 0.2 1>,<at 1 0.4>]}}}, Scale Curve=0.95)
    Scale RGB=<5 5 5>
    ScaleA=true
    ScaleRGB=false
   )
   &FN_RibbonMaterial()
  ]
  rendererClasses [&NiagaraRibbonRendererProperties, &NiagaraMeshRendererProperties]
  renderers [
   {rendererClass=&NiagaraRibbonRendererProperties}
   {rendererClass=&NiagaraRibbonRendererProperties, rendererIndex=1}
   {rendererClass=&NiagaraMeshRendererProperties, rendererIndex=2}
  ]
 }
 Sprinkles {
  emitterSpawnScript [&FN_PreserveSkydiveUserVariables() {module=PreserveUserVariables}]
  emitterUpdateScript [
   &EmitterState(Life Cycle Mode=::NewEnumerator1, Loop Behavior=::NewEnumerator1, Loop Delay=0.03, Loop Duration=0.3, UseLoopDelay=true)
   &SpawnPerUnit(Module.Use Max Movement Threshold=true, Module.Use Movement Tolerance=false, Spawn Per Unit=50, Use Spawn Probability=false)
  ]
  particleSpawnScript [
   &FN_SetSpawnGroup(Number of Spawn Groups=2)
   /Niagara/Modules/Spawn/Initialization/V2/InitializeParticle(
    Color=&ColorFromCurve(
     ColorCurve=DataInterface() {propertyValues {
       AlphaCurve {keys [<at 0 1 RCIM_Cubic>,<at 1 1 RCIM_Cubic>]}
       BlueCurve {keys [<at 0 0.75 RCIM_Cubic>,<at 0.3 0.75>,<at 0.65 1>,<at 1 0 RCIM_Cubic>]}
       GreenCurve {keys [<at 0 0.0302 RCIM_Cubic>,<at 0.3 0.0084>,<at 0.65 0.1415>,<at 1 0.7558 RCIM_Cubic>]}
       RedCurve {keys [<at 0 0.6988 RCIM_Cubic>,<at 0.3 0.175>,<at 0.65 0>,<at 1 1 RCIM_Cubic>]}
      }}
     CurveIndex=&RandomRangeFloat(Minimum=0, Maximum=1)
    )
    Lifetime Max=1
    Lifetime Min=0.5
    Lifetime Mode=::NewEnumerator1
    Mesh Scale=&Multiply_VectorByFloat(Vector=<0.3 0.3 1>, Float=&RandomRangeFloat(Minimum=0.2, Maximum=0.3))
    Mesh Scale Mode=::NewEnumerator4
    UsePositionOffset=false
    Write Material Random=true
   )
   &SkeletalMeshLocation(
    Bone / Socket Index=1
    MaskSpawnGroup=false
    Mesh Sampling Type=::NewEnumerator5
    Skeletal Mesh=$User.SkeletalMesh
    Socket Sampling Mode=::NewEnumerator5
   )
   &BoxLocation(
    Box Origin:NiagaraPosition=$Output.SkeletalMeshLocation.SampledPositionWithoutOffset
    Box Size=<7 1 25>
    MaskSpawnGroup=false
    Module.Override Seed=false
    Rotate Box=false
    Surface Only=false
   )
   &AddVelocity(Velocity=&UniformRangedVector(Minimum=<0 -2 0>, Maximum=<0 5 0>))
   &AddVelocityInCone(
    Cone Angle=75
    Module.Use Velocity Falloff On Cone Axis=false
    Velocity Strength=&UniformRangedFloat(Minimum=1, Maximum=200)
   )
   &InheritVelocity(Module.Limit Speed=true, Velocity Scale=<1 1 0.5>) {module=InheritVelocity001}
   &AddVelocityFromPoint(
    Module.Use Falloff Distance=false
    Velocity Origin:Vector3f=&SimulationPosition()
    Velocity Strength=&UniformRangedFloat(Minimum=10, Maximum=100)
   )
   /SetVariables(Particles.MeshScaleRandom:NiagaraFloat=&UniformRangedFloat(Minimum=1, Maximum=2)) {isSetParameters=true, module=SetVariables_#2}
   &InitialMeshOrientation(Use Rotation Vector=false)
  ]
  particleUpdateScript [
   &ParticleState()
   &GravityForce(Gravity=<0 0 -700>)
   &SolveForcesAndVelocity(Clamp Velocity=false, Limit Acceleration=false)
   &ScaleMeshSize(Scale Factor=&VectorFromFloat(Value=&FloatFromCurve(
      FloatCurve=DataInterface() {propertyValues {Curve {keys [<at 0 1 RCIM_Cubic>,<at 0.2048 0.8914>,<at 1 0 RCIM_Cubic>]}}}
      Scale Curve=&Multiply_Float(A=$Particles.MeshScaleRandom, B=1.2)
     )))
   &ScaleColor(
    Scale Alpha=&FloatFromCurve(FloatCurve=DataInterface() {propertyValues {Curve {keys [
         {interpMode=RCIM_Cubic, tangentMode=RCTM_User, time=0, value=1}
         {arriveTangent=-0.3941, interpMode=RCIM_Cubic, leaveTangent=-0.3941, tangentMode=RCTM_User, time=1, value=0}
        ]}}})
    Scale RGB=&VectorFromFloat(Value=&FloatFromCurve(FloatCurve=DataInterface() {propertyValues {Curve {keys [
          {interpMode=RCIM_Cubic, tangentMode=RCTM_Break, time=0.6, value=1}
          {arriveTangent=-0.9546, interpMode=RCIM_Cubic, tangentMode=RCTM_Break, time=1, value=0}
         ]}}}))
    ScaleA=false
    ScaleRGB=true
   )
   &ScaleMeshSizeBySpeed(
    Max Scale Factor=<2 1.5 1>
    Speed Threshold:NiagaraFloat=&FloatFromCurve(FloatCurve=DataInterface() {propertyValues {Curve {keys [
         {arriveTangent=2294.4238, interpMode=RCIM_Cubic, leaveTangent=2294.4238, tangentMode=RCTM_User, time=0, value=150}
         {arriveTangent=78.1625, interpMode=RCIM_Cubic, leaveTangent=78.3263, tangentMode=RCTM_User, time=1, value=2000}
        ]}}})
   )
   &UpdateMeshOrientation(
    Rotation Rate=&RandomRangeFloat(Minimum=-3, Maximum=3)
    Rotation Vector=&RandomRangeVector(Minimum=<0 0 0>, Maximum=<1 1 1>)
   )
  ]
  rendererClasses [&NiagaraMeshRendererProperties]
  renderers [{rendererClass=&NiagaraMeshRendererProperties, rendererIndex=0}]
 }
}
summary {
 systemName=NS_Pickaxe_FNBirthday6_Trail
 emitters [
  {
   bEnabled=false
   emitterName=FrostingTrail
   rendererClasses [&NiagaraRibbonRendererProperties, &NiagaraMeshRendererProperties]
  }
  {emitterName=Drops, rendererClasses [&NiagaraMeshRendererProperties]}
  {emitterName=Sprinkles, rendererClasses [&NiagaraMeshRendererProperties]}
  {emitterName=Cake, rendererClasses [&NiagaraMeshRendererProperties]}
  {
   bEnabled=false
   emitterName=FrostingTrail001
   rendererClasses [&NiagaraRibbonRendererProperties, &NiagaraMeshRendererProperties]
  }
  {emitterName=FN_AnimTrail, rendererClasses [&NiagaraRibbonRendererProperties]}
  {emitterName=FN_AnimTrail001, rendererClasses [&NiagaraRibbonRendererProperties]}
 ]
 userVariables [
  User.bIsFrontEndPreview {type=NiagaraBool}
  User.PlayerSpeed {type=NiagaraFloat, defaultValue=0, description=""}
  User.RibbonIDOffset {type=NiagaraInt32}
  User.SkeletalMesh {
   type=NiagaraDataInterfaceSkeletalMesh
   defaultValue {
    struct=NiagaraExt_VariableValue_DataInterface
    value {dataInterface=&self:NiagaraDataInterfaceSkeletalMesh_0}
   }
  }
 ]
}
