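# Physics asset for the FN_Birthday_Male pickaxe skeletal mesh, defining simulation bodies for the cake pieces on the weapon.
# Each body is a small capsule: one on the b_AR_Root bone, plus front and back capsules on dyn_ bones for bottom, two middle and top cake tiers.
# Every cake body is constrained directly to the root in a star layout rather than a chain, with limited swing and twist so each piece moves independently.
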PhysicsAsset /BRCosmetics/Weapons/FORT_Melee/Pickaxe_FN_Birthday_Male/Meshes/FN_Birthday_Male_Axe_Physics
bodies {
 b_AR_Root [{shapeType=Capsule, length=2, radius=2, shapeName=b_AR_Root_capsule}]
 dyn_bottom_cake_back [{shapeType=Capsule, length=2, radius=5, shapeName=dyn_bottom_cake_back_capsule}]
 dyn_bottom_cake_front [{shapeType=Capsule, length=2, radius=5, shapeName=dyn_bottom_cake_front_capsule}]
 dyn_middle1_cake_back [{shapeType=Capsule, length=2, radius=5, shapeName=dyn_middle1_cake_back_capsule}]
 dyn_middle1_cake_front [{shapeType=Capsule, length=2, radius=5, shapeName=dyn_middle1_cake_front_capsule}]
 dyn_middle2_cake_back [{shapeType=Capsule, length=2, radius=5, shapeName=dyn_middle2_cake_back_capsule}]
 dyn_middle2_cake_front [{shapeType=Capsule, length=2, radius=5, shapeName=dyn_middle2_cake_front_capsule}]
 dyn_top_cake_back [{shapeType=Capsule, length=2, radius=5, shapeName=dyn_top_cake_back_capsule}]
 dyn_top_cake_front [{shapeType=Capsule, length=2, radius=5, shapeName=dyn_top_cake_front_capsule}]
}
constraints [
 {bone1Name=dyn_bottom_cake_back, bone2Name=b_AR_Root, swing1LimitDegrees=15, swing2LimitDegrees=45, twistLimitDegrees=10}
 {bone1Name=dyn_bottom_cake_front, bone2Name=b_AR_Root, swing1LimitDegrees=15, swing2LimitDegrees=45, twistLimitDegrees=10}
 {bone1Name=dyn_middle2_cake_front, bone2Name=b_AR_Root, swing1LimitDegrees=15, swing2LimitDegrees=45, twistLimitDegrees=10}
 {bone1Name=dyn_middle2_cake_back, bone2Name=b_AR_Root, swing1LimitDegrees=15, swing2LimitDegrees=45, twistLimitDegrees=10}
 {bone1Name=dyn_middle1_cake_back, bone2Name=b_AR_Root, swing1LimitDegrees=15, swing2LimitDegrees=45, twistLimitDegrees=10}
 {bone1Name=dyn_middle1_cake_front, bone2Name=b_AR_Root, swing1LimitDegrees=15, swing2LimitDegrees=45, twistLimitDegrees=10}
 {bone1Name=dyn_top_cake_front, bone2Name=b_AR_Root, swing1LimitDegrees=15, swing2LimitDegrees=45, twistLimitDegrees=15}
 {bone1Name=dyn_top_cake_back, bone2Name=b_AR_Root, swing1LimitDegrees=15, swing2LimitDegrees=45, twistLimitDegrees=10}
]
